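# Physics asset for the F_MED_Platinum_Blue female player body, covering the pelvis, thighs and a dynamic baton on each side.
# Capsule bodies chain from pelvis to each thigh, to a small master baton joint, then to the dynamic baton joint.
# The final baton constraints allow a moderate swing on one axis while nearly locking the other swing and twist, giving each baton a mostly single-plane sway.
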PhysicsAsset /BRCosmetics/Characters/Player/Female/Medium/Bodies/F_MED_Platinum_Blue/Meshes/F_MED_Platinum_Blue_Physics
bodies {
 L_dyn_baton_jnt [{shapeType=Capsule, length=29, radius=5, shapeName=L_dyn_baton_jnt_capsule, center=<-11.939 -2.1237 0>, rotation=<rot 90 90 89.9999>}]
 R_dyn_baton_jnt [{shapeType=Capsule, length=29, radius=5, shapeName=R_dyn_baton_jnt_capsule, center=<11.9215 1.6535 0>, rotation=<rot 90 90 89.9999>}]
 master_L_baton_jnt [{shapeType=Capsule, length=1, radius=1, shapeName=master_L_baton_jnt_capsule}]
 master_R_baton_jnt [{shapeType=Capsule, length=1, radius=1, shapeName=master_R_baton_jnt_capsule}]
 pelvis [{shapeType=Capsule, length=5, radius=11, shapeName=pelvis_capsule}]
 thigh_l [{shapeType=Capsule, length=28, radius=5, shapeName=thigh_l_capsule, center=<-19.9928 0.0001 0>, rotation=<rot 90 90 90.0004>}]
 thigh_r [{shapeType=Capsule, length=28, radius=5, shapeName=thigh_r_capsule, center=<19.8681 -0.0001 0>, rotation=<rot 90 179.99 179.9904>}]
}
constraints [
 {bone1Name=thigh_r, bone2Name=pelvis, swing1LimitDegrees=45, swing2LimitDegrees=45, twistLimitDegrees=45}
 {bone1Name=thigh_l, bone2Name=pelvis, swing1LimitDegrees=45, swing2LimitDegrees=45, twistLimitDegrees=45}
 {bone1Name=master_L_baton_jnt, bone2Name=thigh_l, swing1LimitDegrees=45, swing2LimitDegrees=45, twistLimitDegrees=45}
 {bone1Name=L_dyn_baton_jnt, bone2Name=master_L_baton_jnt, swing1LimitDegrees=2, swing2LimitDegrees=25, twistLimitDegrees=1}
 {bone1Name=master_R_baton_jnt, bone2Name=thigh_r, swing1LimitDegrees=45, swing2LimitDegrees=45, twistLimitDegrees=45}
 {bone1Name=R_dyn_baton_jnt, bone2Name=master_R_baton_jnt, swing1LimitDegrees=2, swing2LimitDegrees=25, twistLimitDegrees=1}
]
