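# Post-processes the Platinum Blue female body cosmetic's pose, correcting its batons, thigh straps and harness.
# Copies the attached parent's pose through a deform Control Rig, then applies weighted bone constraints, crouch-scaled additive offsets and pelvis-space RigidBody physics.
# Post-physics offsets adjust the batons and right harness, some only with backbling in the front end.
# The update event caches those states and hides a material element for cape or long backbling.
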AnimBlueprint /BRCosmetics/Characters/Player/Female/Medium/Bodies/F_MED_Platinum_Blue/Meshes/F_MED_Platinum_Blue_AnimBP
target_skeleton=&/F_MED_Platinum_Blue_Skeleton
graphs [
 AnimGraph {nodes [
   n1 = Control Rig(Source=Copy Pose From Mesh() {bUseAttachedParent=true}) {
    alphaBoolBlend {blendOption=Linear}
    controlRigAssetReference {blueprintRigClass=&/F_MED_DeformRig_PlatinumBlue_ControlRig.F_MED_DeformRig_PlatinumBlue_ControlRig_C}
   }
   n2 = Get IsCrouching() -> (IsCrouching:double)
   n3 = "Constraint - master_L_baton_jnt (PPPP)"!(
    ComponentPose=Local To Component(LocalPose=@n1)
    Alpha=SelectFloat(A=1, B=@n2, bPickA=true)
    ConstraintWeights_0=0.3086
    ConstraintWeights_1:float=0.3559
    ConstraintWeights_2:float=0.0015
    ConstraintWeights_3:float=0.334
   ) -> (Pose) {
    alphaBoolBlend {blendOption=Linear}
    boneToModify {boneName=master_L_baton_jnt}
    constraintSetup [
     {targetBone {boneName=pelvis}}
     {targetBone {boneName=thigh_twist_01_l}}
     {targetBone {boneName=thigh_l}}
     {targetBone {boneName=deform_glute_l}}
    ]
    constraintWeights [1,1,1,1]
   }
   n4 = "Constraint - master_L_baton_jnt (PPPPPPPP)"!(
    ComponentPose=@n3
    Alpha=@n2
    ConstraintWeights_0=0.2313
    ConstraintWeights_1:float=0
    ConstraintWeights_2:float=0.0001
    ConstraintWeights_3:float=0.1085
    ConstraintWeights_4:float=0.5175
    ConstraintWeights_5:float=0.0212
    ConstraintWeights_6:float=0
    ConstraintWeights_7:float=0.1214
   ) -> (Pose) {
    alphaBoolBlend {blendOption=Linear}
    boneToModify {boneName=master_L_baton_jnt}
    constraintSetup [
     {targetBone {boneName=pelvis}}
     {targetBone {boneName=spine_01}}
     {targetBone {boneName=spine_02}}
     {targetBone {boneName=thigh_l}}
     {targetBone {boneName=thigh_twist_01_l}}
     {targetBone {boneName=deform_groin_l}}
     {targetBone {boneName=thigh_twist_01_r}}
     {targetBone {boneName=deform_glute_base_l}}
    ]
    constraintWeights [1,1,1,1,1,1,1,1]
   }
   n5 = "Constraint - master_R_baton_jnt (PPPP)"!(
    ComponentPose=@n4
    Alpha=1
    ConstraintWeights_0=0.1936
    ConstraintWeights_1:float=0.4953
    ConstraintWeights_2:float=0.0266
    ConstraintWeights_3:float=0.2845
   ) -> (Pose) {
    alphaBoolBlend {blendOption=Linear}
    boneToModify {boneName=master_R_baton_jnt}
    constraintSetup [
     {targetBone {boneName=pelvis}}
     {targetBone {boneName=thigh_twist_01_r}}
     {targetBone {boneName=thigh_r}}
     {targetBone {boneName=deform_glute_r}}
    ]
    constraintWeights [1,1,1,1]
   }
   n6 = "Transform (Modify) Bone - Bone: master_L_baton_jnt"(ComponentPose=@n5, Alpha=@n2, Translation="-1,2,0", Rotation="0,0,0", Scale="1,1,1") {
    translationMode=BMM_Additive
    translationSpace=BCS_BoneSpace
    alphaBoolBlend {blendOption=Linear}
    boneToModify {boneName=master_L_baton_jnt}
    translation=<-1 2>
   }
   n7 = "Transform (Modify) Bone - Bone: master_R_baton_jnt"(ComponentPose=@n6, Alpha=@n2, Translation="0,-1,0", Rotation="0,0,0", Scale="1,1,1") {
    translationMode=BMM_Additive
    translationSpace=BCS_BoneSpace
    alphaBoolBlend {blendOption=Linear}
    boneToModify {boneName=master_R_baton_jnt}
    translation {y=-1}
   }
   n8 = "Constraint - l_thighstrap_helper_jnt (PPPPPPPP)"!(
    ComponentPose=@n7
    Alpha=1
    ConstraintWeights_0=0.3545
    ConstraintWeights_1:float=0
    ConstraintWeights_2:float=0
    ConstraintWeights_3:float=0.0351
    ConstraintWeights_4:float=0.372
    ConstraintWeights_5:float=0.2365
    ConstraintWeights_6:float=0.0002
    ConstraintWeights_7:float=0.0016
   ) -> (Pose) {
    alphaBoolBlend {blendOption=Linear}
    boneToModify {boneName=l_thighstrap_helper_jnt}
    constraintSetup [
     {targetBone {boneName=pelvis}}
     {targetBone {boneName=spine_01}}
     {targetBone {boneName=spine_02}}
     {targetBone {boneName=thigh_l}}
     {targetBone {boneName=thigh_twist_01_l}}
     {targetBone {boneName=deform_groin_l}}
     {targetBone {boneName=thigh_twist_01_r}}
     {targetBone {boneName=deform_glute_base_l}}
    ]
    constraintWeights [1,1,1,1,1,1,1,1]
   }
   n9 = "Constraint - r_thighstrap_helper_jnt (PPPPPPPP)"!(
    ComponentPose=@n8
    Alpha=1
    ConstraintWeights_0=0.3545
    ConstraintWeights_1:float=0
    ConstraintWeights_2:float=0
    ConstraintWeights_3:float=0.0002
    ConstraintWeights_4:float=0.0351
    ConstraintWeights_5:float=0.372
    ConstraintWeights_6:float=0.2365
    ConstraintWeights_7:float=0.0016
   ) -> (Pose) {
    alphaBoolBlend {blendOption=Linear}
    boneToModify {boneName=r_thighstrap_helper_jnt}
    constraintSetup [
     {targetBone {boneName=pelvis}}
     {targetBone {boneName=spine_01}}
     {targetBone {boneName=spine_02}}
     {targetBone {boneName=thigh_twist_01_l}}
     {targetBone {boneName=thigh_r}}
     {targetBone {boneName=thigh_twist_01_r}}
     {targetBone {boneName=deform_groin_r}}
     {targetBone {boneName=deform_glute_r}}
    ]
    constraintWeights [1,1,1,1,1,1,1,1]
   }
   n10 = "Constraint - l_lower_thighstrap_helper_jnt (PPPPPPPP)"!(
    ComponentPose=@n9
    Alpha=1
    ConstraintWeights_0=0.1539
    ConstraintWeights_1:float=0
    ConstraintWeights_2:float=0.0001
    ConstraintWeights_3:float=0.2161
    ConstraintWeights_4:float=0.5373
    ConstraintWeights_5:float=0.0336
    ConstraintWeights_6:float=0
    ConstraintWeights_7:float=0.0589
   ) -> (Pose) {
    alphaBoolBlend {blendOption=Linear}
    boneToModify {boneName=l_lower_thighstrap_helper_jnt}
    constraintSetup [
     {targetBone {boneName=pelvis}}
     {targetBone {boneName=spine_01}}
     {targetBone {boneName=spine_02}}
     {targetBone {boneName=thigh_l}}
     {targetBone {boneName=thigh_twist_01_l}}
     {targetBone {boneName=deform_groin_l}}
     {targetBone {boneName=thigh_twist_01_r}}
     {targetBone {boneName=deform_glute_base_l}}
    ]
    constraintWeights [1,1,1,1,1,1,1,1]
   }
   n11 = "Constraint - r_lower_thighstrap_helper_jnt (PPPPPPPP)"!(
    ComponentPose=@n10
    Alpha=1
    ConstraintWeights_0=0.1539
    ConstraintWeights_1:float=0
    ConstraintWeights_2:float=0.0001
    ConstraintWeights_3:float=0
    ConstraintWeights_4:float=0.2161
    ConstraintWeights_5:float=0.5373
    ConstraintWeights_6:float=0.0336
    ConstraintWeights_7:float=0.0589
   ) -> (Pose) {
    alphaBoolBlend {blendOption=Linear}
    boneToModify {boneName=r_lower_thighstrap_helper_jnt}
    constraintSetup [
     {targetBone {boneName=pelvis}}
     {targetBone {boneName=spine_01}}
     {targetBone {boneName=spine_02}}
     {targetBone {boneName=thigh_twist_01_l}}
     {targetBone {boneName=thigh_r}}
     {targetBone {boneName=thigh_twist_01_r}}
     {targetBone {boneName=deform_groin_r}}
     {targetBone {boneName=deform_glute_r}}
    ]
    constraintWeights [1,1,1,1,1,1,1,1]
   }
   n12 = "Transform (Modify) Bone - Bone: r_thighstrap_helper_jnt"(ComponentPose=@n11, Alpha=@n2, Translation="-2.276,-1.905,3.704", Rotation="20.012,-20.013,-0.030", Scale="1,1,1") {
    rotationMode=BMM_Additive
    rotationSpace=BCS_BoneSpace
    translationMode=BMM_Additive
    translationSpace=BCS_BoneSpace
    alphaBoolBlend {blendOption=Linear}
    boneToModify {boneName=r_thighstrap_helper_jnt}
    rotation=<rot 20.012 -20.013 -0.03>
    translation=<-2.276 -1.905 3.704>
   }
   n13 = "Transform (Modify) Bone - Bone: r_lower_thighstrap_helper_jnt"(ComponentPose=@n12, Alpha=@n2, Translation="-1.749,-2.035,-1.074", Rotation="9.971,0.024,-0.007", Scale="1,1,1") {
    rotationMode=BMM_Additive
    rotationSpace=BCS_BoneSpace
    translationMode=BMM_Additive
    translationSpace=BCS_BoneSpace
    alphaBoolBlend {blendOption=Linear}
    boneToModify {boneName=r_lower_thighstrap_helper_jnt}
    rotation=<rot 9.971 0.024 -0.0071>
    translation=<-1.749 -2.035 -1.074>
   }
   n14 = "Transform (Modify) Bone - Bone: l_thighstrap_helper_jnt"(ComponentPose=@n13, Alpha=@n2, Translation="4.481,3.009,-5.433", Rotation="20.418,6.018,-1.529", Scale="1,1,1") {
    rotationMode=BMM_Additive
    rotationSpace=BCS_BoneSpace
    translationMode=BMM_Additive
    translationSpace=BCS_BoneSpace
    alphaBoolBlend {blendOption=Linear}
    boneToModify {boneName=l_thighstrap_helper_jnt}
    rotation=<rot 20.418 6.018 -1.529>
    translation=<4.481 3.009 -5.433>
   }
   n15 = "Transform (Modify) Bone - Bone: l_lower_thighstrap_helper_jnt"(ComponentPose=@n14, Alpha=@n2, Translation="-0.751,-0.000,0.981", Rotation="0.001,-0.005,30.000", Scale="1,1,1") {
    rotationMode=BMM_Additive
    rotationSpace=BCS_BoneSpace
    translationMode=BMM_Additive
    translationSpace=BCS_BoneSpace
    alphaBoolBlend {blendOption=Linear}
    boneToModify {boneName=l_lower_thighstrap_helper_jnt}
    rotation=<rot 0.001 -0.005 29.9999>
    translation {x=-0.751, z=0.981}
   }
   n16 = "Transform (Modify) Bone - Bone: master_L_baton_jnt"(ComponentPose=@n15, Alpha=@n2, Translation="-0.874,-1.727,-0.322", Rotation="0,0,0", Scale="1,1,1") {
    rotationMode=BMM_Additive
    rotationSpace=BCS_BoneSpace
    translationMode=BMM_Additive
    translationSpace=BCS_BoneSpace
    alphaBoolBlend {blendOption=Linear}
    boneToModify {boneName=master_L_baton_jnt}
    translation=<-0.874 -1.727 -0.322>
   }
   n17 = RigidBody(ComponentPose=@n16) {
    bClampLinearTranslationLimitToRefPose=true
    bOverrideWorldGravity=true
    overridePhysicsAsset=&/F_MED_Platinum_Blue_Physics
    simulationSpace=BaseBoneSpace
    alphaBoolBlend {blendOption=Linear}
    baseBoneRef {boneName=pelvis}
    componentAppliedLinearAccClamp=<3000 3000 3000>
    componentLinearAccScale=<0.3 0.2 0.2>
    componentLinearVelScale=<0.3 0.2 0.2>
    overrideWorldGravity {z=-980}
   }
   n18 = AND Boolean(A=Get BackblingEquipped() -> (BackblingEquipped:bool), B=Get FrontEnd() -> (FrontEnd:bool))
   n19 = "Transform (Modify) Bone - Bone: R_dyn_baton_jnt"!(ComponentPose=@n17, bAlphaBoolEnabled:bool=@n18, Translation="0,0,0", Rotation="10.000,0.000,0", Scale="1,1,1") -> (Pose) {
    alphaInputType=Bool
    rotationMode=BMM_Additive
    rotationSpace=BCS_BoneSpace
    alphaBoolBlend {blendOption=Linear}
    boneToModify {boneName=R_dyn_baton_jnt}
    rotation {pitch=10}
   }
   n20 = "Transform (Modify) Bone - Bone: L_dyn_baton_jnt"!(ComponentPose=@n19, bAlphaBoolEnabled:bool=@n18, Translation="0,0,0", Rotation="19.836,0.156,1.795", Scale="1,1,1") -> (Pose) {
    alphaInputType=Bool
    rotationMode=BMM_Additive
    rotationSpace=BCS_BoneSpace
    alphaBoolBlend {blendOption=Linear}
    boneToModify {boneName=L_dyn_baton_jnt}
    rotation=<rot 19.836 0.156 1.795>
   }
   n21 = "Transform (Modify) Bone - Bone: r_harness_helper_jnt"!(ComponentPose=@n20, bAlphaBoolEnabled:bool=True, Translation="-2.492,-3.516,-0.630", Rotation="7.313,29.910,29.752", Scale="1,1,1") -> (Pose) {
    alphaInputType=Bool
    rotationMode=BMM_Additive
    rotationSpace=BCS_BoneSpace
    translationMode=BMM_Additive
    alphaBoolBlend {blendOption=Linear}
    boneToModify {boneName=r_harness_helper_jnt}
    rotation=<rot 7.313 29.91 29.752>
    translation=<-2.492 -3.516 -0.63>
   }
   n22 = "Transform (Modify) Bone - Bone: R_dyn_baton_jnt"(ComponentPose=@n21, bAlphaBoolEnabled=True, Translation="-0.683,2.619,-0.386", Rotation="17.068,30.291,2.318", Scale="1,1,1") {
    alphaInputType=Bool
    rotationMode=BMM_Additive
    rotationSpace=BCS_BoneSpace
    translationMode=BMM_Additive
    alphaBoolBlend {blendOption=Linear}
    boneToModify {boneName=R_dyn_baton_jnt}
    rotation=<rot 17.0681 30.291 2.3179>
    translation=<-0.683 2.619 -0.386>
   }
   Output Pose(Result=Component To Local(ComponentPose=@n22)) {layerGroup=None}
  ]}
 EventGraph {nodes [
   Event BlueprintUpdateAnimation()
   > Set IsCrouching(IsCrouching:double="To Float (Boolean)"(InBool=Get bIsCrouching())) -> (Output_Get:double)
   > n1 = Get Dynamic Material Instance(SkeletalMeshComponent=GetPartSkeletalMeshComponent())
   n2 = CheckForBackbling()
   > SetScalarParameterValue(self=@n1, ParameterName=Hide Element 01, Value="To Float (Boolean)"(InBool=OR Boolean(A=@n2.bIsBackblingACape, B=@n2.bIsBackblingLong)))
   > Set FrontEnd(FrontEnd:bool=Get bIsInFrontEnd()) -> (Output_Get:bool)
   > Set BackblingEquipped(BackblingEquipped:bool=OR Boolean(A=@n2, B=@n2.bIsBackblingACape)) -> (Output_Get:bool)
  ]}
]
variables [
 BackblingEquipped {type=boolean}
 FrontEnd {type=boolean}
 IsCrouching {type=number}
]
