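# Physics asset for the Platinum Blue medium female FaceAcc/Backbling mesh, providing capsule bodies for dynamic hair and upper-body collision.
# Bangs, side locks, left/right chains and three back chains hang from master joints under master_hair_jnt, which attaches to spine_03.
# Chain segments mostly use tight swing and twist limits, while most body constraints are angularly free.
# Extra angularly free constraints cross-link the outer back-chain tips to the center tip.
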PhysicsAsset /BRCosmetics/Characters/Player/Female/Medium/Bodies/F_MED_Platinum_Blue/Meshes/Parts/F_MED_Platinum_Blue_FaceAcc_Backbling_Physics
bodies {
 clavicle_l [{
   shapeType=Capsule
   length=14
   radius=5.625
   shapeName=clavicle_l_capsule
   center=<11.3065 -0.1722 2.0772>
   rotation=<rot 83.6478 -164.1804 -174.1237>
  }]
 clavicle_r [{
   shapeType=Capsule
   length=14
   radius=5.625
   shapeName=clavicle_l_capsule
   center=<-11.6326 0.1117 -1.7248>
   rotation=<rot 83.6478 -164.1807 5.876>
  }]
 dyn_L_chain_1_jnt [{shapeType=Capsule, length=3, radius=3, shapeName=dyn_L_chain_1_jnt_capsule, center=<3.4033 2.1866 0>}]
 dyn_L_chain_2_jnt [{shapeType=Capsule, length=2.25, radius=3.75, shapeName=dyn_L_chain_2_jnt_capsule, center=<4.0218 1.6062 0>}]
 dyn_L_chain_3_jnt [{shapeType=Capsule, length=4.5, radius=3, shapeName=dyn_L_chain_3_jnt_capsule, center=<4.3732 1.5544 0>, rotation=<rot 90 0 0>}]
 dyn_L_side_hair_jnt [{shapeType=Capsule, length=2, radius=2, shapeName=dyn_L_side_hair_jnt_capsule, center=<2.5536 1.0626 0>, rotation=<rot 90 180 179.9998>}]
 dyn_R_chain_1_jnt [{shapeType=Capsule, length=1.75, radius=4.25, shapeName=dyn_R_chain_1_jnt_capsule, center=<2.325 0.7658 0>}]
 dyn_R_chain_2_jnt [{shapeType=Capsule, length=3, radius=3.25, shapeName=dyn_R_chain_2_jnt_capsule, center=<3.8784 1.9989 0>}]
 dyn_R_chain_3_jnt [{shapeType=Capsule, length=6, radius=3, shapeName=dyn_R_chain_3_jnt_capsule, center=<6.72 1.6247 0>, rotation=<rot 90 0 0>}]
 dyn_R_chain_flip_jnt [{shapeType=Capsule, length=3, radius=2, shapeName=dyn_R_chain_flip_jnt_capsule, center=<2.3903 1.8039 -1.38>, rotation=<rot 60.0001 0.0004 0>}]
 dyn_R_side_hair_jnt [{shapeType=Capsule, length=2, radius=2, shapeName=dyn_R_side_hair_jnt_capsule, center=<2.5536 1.0626 0>, rotation=<rot 90 180 179.9998>}]
 dyn_back_chain_1_jnt [{shapeType=Capsule, length=1.5, radius=4.5, shapeName=dyn_back_chain_1_jnt_capsule, center=<4.7416 -4.3582 0>}]
 dyn_back_chain_2_jnt [{shapeType=Capsule, length=2.25, radius=3.75, shapeName=dyn_back_chain_2_jnt_capsule, center=<4.1962 -4.42 0>}]
 dyn_back_chain_3_jnt [{shapeType=Capsule, length=2.25, radius=3.75, shapeName=dyn_back_chain_3_jnt_capsule, center=<4.1962 -2.9778 0>}]
 dyn_back_chain_4_jnt [{shapeType=Capsule, length=4.75, radius=4.75, shapeName=dyn_back_r_chain_4_jnt_capsule, center=<6.6917 -2.3775 0.1018>, rotation=<rot 90 0 0>}]
 dyn_back_l_chain_1_jnt [{shapeType=Capsule, length=1.5, radius=4.5, shapeName=dyn_back_l_chain_1_jnt_capsule, center=<4.7416 -2.5559 0>}]
 dyn_back_l_chain_2_jnt [{shapeType=Capsule, length=2.25, radius=3.75, shapeName=dyn_back_l_chain_2_jnt_capsule, center=<3.1429 -3.0389 0>}]
 dyn_back_l_chain_3_jnt [{shapeType=Capsule, length=2.25, radius=3.75, shapeName=dyn_back_l_chain_3_jnt_capsule, center=<5.3144 -2.9778 1.5598>}]
 dyn_back_l_chain_4_jnt [{shapeType=Capsule, length=4.75, radius=4.75, shapeName=dyn_back_r_chain_4_jnt_capsule, center=<6.2002 -2.3775 3.8566>, rotation=<rot 90 0 0>}]
 dyn_back_r_chain_1_jnt [{shapeType=Capsule, length=1.5, radius=4.5, shapeName=dyn_back_r_chain_1_jnt_capsule, center=<4.7416 -3.9365 0>}]
 dyn_back_r_chain_2_jnt [{shapeType=Capsule, length=2, radius=4, shapeName=dyn_back_r_chain_2_jnt_capsule, center=<4.1962 -3.1782 0>}]
 dyn_back_r_chain_3_jnt [{shapeType=Capsule, length=2.25, radius=3.75, shapeName=dyn_back_r_chain_3_jnt_capsule, center=<4.2914 -2.9778 -1.5131>}]
 dyn_back_r_chain_4_jnt [{shapeType=Capsule, length=4.75, radius=4.75, shapeName=dyn_back_r_chain_4_jnt_capsule, center=<7.1368 -2.3775 0.1018>, rotation=<rot 90 0 0>}]
 dyn_bangs_jnt [{
   shapeType=Capsule
   length=5
   radius=2.5
   shapeName=dyn_bangs_jnt_capsule
   center=<1.1541 0.4805 -0.4148>
   rotation=<rot -90 -179.6835 179.6837>
  }]
 head [{shapeType=Capsule, length=6.1965, radius=6.5617, shapeName=head_capsule, center=<2.5967 2.5691 0>, rotation=<rot 90 0 0>}]
 master_L_chain_jnt [{shapeType=Capsule, length=1, radius=1, shapeName=master_L_chain_jnt_capsule}]
 master_R_chain_jnt [{shapeType=Capsule, length=1, radius=1, shapeName=master_R_chain_jnt_capsule}]
 master_back_chain_jnt [{shapeType=Capsule, length=0.505, radius=0.505, shapeName=master_back_chain_jnt_capsule, rotation=<rot 0 -90 0>}]
 master_hair_jnt [{shapeType=Capsule, length=1, radius=1, shapeName=master_hair_jnt_capsule}]
 neck_01 [{shapeType=Capsule, length=4.6965, radius=6.4367, shapeName=head_capsule, center=<2.5967 2.0576 0>, rotation=<rot 90 0 0>}]
 spine_03 [{shapeType=Capsule, length=5.25, radius=11, shapeName=spine_04_capsule, center=<-0.3569 2.6072 0>}]
 spine_04 [{shapeType=Capsule, length=5, radius=11.25, shapeName=spine_04_capsule, center=<0.4833 1.9817 0>}]
 spine_05 [
  {shapeType=Capsule, length=11.25, radius=8.25, shapeName=spine_05_capsule, center=<2.6959 -1.5257 0>}
  {shapeType=Capsule, length=56.75, radius=10.5, shapeName=spine_05_capsule, center=<9.981 9.8738 0>}
 ]
 upperarm_l [{
   shapeType=Capsule
   length=29
   radius=5.125
   shapeName=clavicle_l_capsule
   center=<10.4528 -1.6688 -0.5227>
   rotation=<rot 90 172.875 -167.1253>
  }]
 upperarm_r [{
   shapeType=Capsule
   length=29
   radius=5.125
   shapeName=clavicle_l_capsule
   center=<-10.8361 1.9745 -0.9313>
   rotation=<rot 90 -177.6141 -157.6145>
  }]
}
constraints [
 {bone1Name=head, bone2Name=neck_01, swing1LimitDegrees=45, swing2LimitDegrees=45, twistLimitDegrees=45}
 {bone1Name=dyn_bangs_jnt, bone2Name=master_hair_jnt, swing1LimitDegrees=45, swing2LimitDegrees=45, twistLimitDegrees=45}
 {bone1Name=dyn_R_side_hair_jnt, bone2Name=master_hair_jnt, swing1LimitDegrees=45, swing2LimitDegrees=45, twistLimitDegrees=45}
 {bone1Name=dyn_L_side_hair_jnt, bone2Name=master_hair_jnt, swing1LimitDegrees=45, swing2LimitDegrees=45, twistLimitDegrees=45}
 {bone1Name=master_L_chain_jnt, bone2Name=master_hair_jnt, swing1LimitDegrees=45, swing2LimitDegrees=45, twistLimitDegrees=45}
 {bone1Name=dyn_L_chain_1_jnt, bone2Name=master_L_chain_jnt, swing1LimitDegrees=15, swing2LimitDegrees=15, twistLimitDegrees=2}
 {bone1Name=dyn_L_chain_2_jnt, bone2Name=dyn_L_chain_1_jnt, swing1LimitDegrees=15, swing2LimitDegrees=15, twistLimitDegrees=2}
 {bone1Name=dyn_L_chain_3_jnt, bone2Name=dyn_L_chain_2_jnt, swing1LimitDegrees=15, swing2LimitDegrees=15, twistLimitDegrees=2}
 {bone1Name=master_R_chain_jnt, bone2Name=master_hair_jnt, swing1LimitDegrees=45, swing2LimitDegrees=45, twistLimitDegrees=45}
 {bone1Name=dyn_R_chain_1_jnt, bone2Name=master_R_chain_jnt, swing1LimitDegrees=45, swing2LimitDegrees=15, twistLimitDegrees=2}
 {bone1Name=dyn_R_chain_2_jnt, bone2Name=dyn_R_chain_1_jnt, swing1LimitDegrees=45, swing2LimitDegrees=15, twistLimitDegrees=2}
 {bone1Name=dyn_R_chain_3_jnt, bone2Name=dyn_R_chain_2_jnt, swing1LimitDegrees=45, swing2LimitDegrees=15, twistLimitDegrees=2}
 {bone1Name=dyn_R_chain_flip_jnt, bone2Name=dyn_R_chain_2_jnt, swing1LimitDegrees=15, swing2LimitDegrees=15, twistLimitDegrees=2}
 {
  bone1Name=neck_01
  bone2Name=spine_05
  swing1LimitDegrees=45
  swing1Motion=Free
  swing2LimitDegrees=45
  swing2Motion=Free
  twistLimitDegrees=45
  twistMotion=Free
 }
 {
  bone1Name=spine_05
  bone2Name=spine_04
  swing1LimitDegrees=45
  swing1Motion=Free
  swing2LimitDegrees=45
  swing2Motion=Free
  twistLimitDegrees=45
  twistMotion=Free
 }
 {
  bone1Name=master_back_chain_jnt
  bone2Name=master_hair_jnt
  swing1LimitDegrees=45
  swing1Motion=Free
  swing2LimitDegrees=45
  swing2Motion=Free
  twistLimitDegrees=45
  twistMotion=Free
 }
 {bone1Name=dyn_back_chain_1_jnt, bone2Name=master_back_chain_jnt, swing1LimitDegrees=10, swing2LimitDegrees=10, twistLimitDegrees=2}
 {bone1Name=dyn_back_chain_2_jnt, bone2Name=dyn_back_chain_1_jnt, swing1LimitDegrees=10, swing2LimitDegrees=10, twistLimitDegrees=2}
 {bone1Name=dyn_back_chain_3_jnt, bone2Name=dyn_back_chain_2_jnt, swing1LimitDegrees=10, swing2LimitDegrees=10, twistLimitDegrees=2}
 {bone1Name=dyn_back_chain_4_jnt, bone2Name=dyn_back_chain_3_jnt, swing1LimitDegrees=10, swing2LimitDegrees=10, twistLimitDegrees=2}
 {bone1Name=dyn_back_r_chain_1_jnt, bone2Name=master_back_chain_jnt, swing1LimitDegrees=10, swing2LimitDegrees=10, twistLimitDegrees=2}
 {bone1Name=dyn_back_r_chain_2_jnt, bone2Name=dyn_back_r_chain_1_jnt, swing1LimitDegrees=10, swing2LimitDegrees=10, twistLimitDegrees=2}
 {bone1Name=dyn_back_r_chain_3_jnt, bone2Name=dyn_back_r_chain_2_jnt, swing1LimitDegrees=10, swing2LimitDegrees=10, twistLimitDegrees=2}
 {bone1Name=dyn_back_r_chain_4_jnt, bone2Name=dyn_back_r_chain_3_jnt, swing1LimitDegrees=10, swing2LimitDegrees=10, twistLimitDegrees=2}
 {bone1Name=dyn_back_l_chain_1_jnt, bone2Name=master_back_chain_jnt, swing1LimitDegrees=10, swing2LimitDegrees=10, twistLimitDegrees=2}
 {bone1Name=dyn_back_l_chain_2_jnt, bone2Name=dyn_back_l_chain_1_jnt, swing1LimitDegrees=10, swing2LimitDegrees=10, twistLimitDegrees=2}
 {bone1Name=dyn_back_l_chain_3_jnt, bone2Name=dyn_back_l_chain_2_jnt, swing1LimitDegrees=10, swing2LimitDegrees=10, twistLimitDegrees=2}
 {bone1Name=dyn_back_l_chain_4_jnt, bone2Name=dyn_back_l_chain_3_jnt, swing1LimitDegrees=10, swing2LimitDegrees=10, twistLimitDegrees=2}
 {
  bone1Name=spine_04
  bone2Name=spine_03
  swing1LimitDegrees=45
  swing1Motion=Free
  swing2LimitDegrees=45
  swing2Motion=Free
  twistLimitDegrees=45
  twistMotion=Free
 }
 {
  bone1Name=clavicle_l
  bone2Name=spine_03
  swing1LimitDegrees=45
  swing1Motion=Free
  swing2LimitDegrees=45
  swing2Motion=Free
  twistLimitDegrees=45
  twistMotion=Free
 }
 {
  bone1Name=upperarm_l
  bone2Name=spine_03
  swing1LimitDegrees=45
  swing1Motion=Free
  swing2LimitDegrees=45
  swing2Motion=Free
  twistLimitDegrees=45
  twistMotion=Free
 }
 {
  bone1Name=master_hair_jnt
  bone2Name=spine_03
  swing1LimitDegrees=45
  swing1Motion=Free
  swing2LimitDegrees=45
  swing2Motion=Free
  twistLimitDegrees=45
  twistMotion=Free
 }
 {
  bone1Name=clavicle_r
  bone2Name=spine_03
  swing1LimitDegrees=45
  swing1Motion=Free
  swing2LimitDegrees=45
  swing2Motion=Free
  twistLimitDegrees=45
  twistMotion=Free
 }
 {
  bone1Name=upperarm_r
  bone2Name=spine_03
  swing1LimitDegrees=45
  swing1Motion=Free
  swing2LimitDegrees=45
  swing2Motion=Free
  twistLimitDegrees=45
  twistMotion=Free
 }
 {
  bone1Name=dyn_back_r_chain_4_jnt
  bone2Name=dyn_back_chain_4_jnt
  swing1LimitDegrees=45
  swing1Motion=Free
  swing2LimitDegrees=45
  swing2Motion=Free
  twistLimitDegrees=45
  twistMotion=Free
 }
 {
  bone1Name=dyn_back_l_chain_4_jnt
  bone2Name=dyn_back_chain_4_jnt
  swing1LimitDegrees=45
  swing1Motion=Free
  swing2LimitDegrees=45
  swing2Motion=Free
  twistLimitDegrees=45
  twistMotion=Free
 }
]
