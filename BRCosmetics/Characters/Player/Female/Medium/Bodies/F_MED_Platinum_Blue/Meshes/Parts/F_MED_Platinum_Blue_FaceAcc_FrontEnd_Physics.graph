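# Physics asset for the front-end face-accessory mesh of the female medium Platinum Blue character, setting up hair dynamics.
# Capsule bodies cover head, neck, upper spine, clavicles and upper arms, plus the hair joints.
# Back, left and right hair chains, a right flip and side locks hang from master joints under the head.
# Hair constraints use limited swing with very tight twist, while body and master-joint constraints leave rotation free.
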PhysicsAsset /BRCosmetics/Characters/Player/Female/Medium/Bodies/F_MED_Platinum_Blue/Meshes/Parts/F_MED_Platinum_Blue_FaceAcc_FrontEnd_Physics
bodies {
 clavicle_l [{
   shapeType=Capsule
   length=6.214
   radius=6.7919
   shapeName=clavicle_r_capsule
   center=<6.606 -0.9847 -1.7639>
   rotation=<rot 83.3792 0.851 -3.3563>
  }]
 clavicle_r [{
   shapeType=Capsule
   length=6.214
   radius=6.7919
   shapeName=clavicle_r_capsule
   center=<-6.8738 0.4534 2.9477>
   rotation=<rot 83.3792 0.851 -3.3563>
  }]
 dyn_L_chain_1_jnt [{shapeType=Capsule, length=1.7289, radius=5.0211, shapeName=dyn_L_chain_1_jnt_capsule, center=<3.3741 2.6776 0>}]
 dyn_L_chain_2_jnt [{shapeType=Capsule, length=1.7289, radius=4.2711, shapeName=dyn_L_chain_2_jnt_capsule, center=<4.1664 3.3408 0>}]
 dyn_L_chain_3_jnt [{shapeType=Capsule, length=5.2997, radius=3.4503, shapeName=dyn_L_chain_3_jnt_capsule, center=<5.3063 -0.7023 0>, rotation=<rot 90 0 0>}]
 dyn_L_side_hair_jnt [{shapeType=Capsule, length=1.7289, radius=4.2711, shapeName=dyn_L_side_hair_jnt_capsule, center=<0 -3 0>}]
 dyn_R_chain_1_jnt [{shapeType=Capsule, length=2.2289, radius=5.2711, shapeName=dyn_R_chain_1_jnt_capsule, center=<5.6272 3.5218 0>}]
 dyn_R_chain_2_jnt [{shapeType=Capsule, length=2.0233, radius=3.9767, shapeName=dyn_R_chain_2_jnt_capsule, center=<4.9381 3.7796 0>}]
 dyn_R_chain_3_jnt [{shapeType=Capsule, length=8.4563, radius=3.1559, shapeName=dyn_R_chain_3_jnt_capsule, center=<7.2066 2.6787 0>, rotation=<rot 90 0 0>}]
 dyn_R_chain_flip_jnt [{
   shapeType=Capsule
   length=4.4861
   radius=2.7639
   shapeName=dyn_R_chain_flip_jnt_capsule
   center=<4.7391 1.5328 -2.7361>
   rotation=<rot 60 0.0001 0>
  }]
 dyn_R_side_hair_jnt [{shapeType=Capsule, length=1.7289, radius=4.2711, shapeName=dyn_R_side_hair_jnt_capsule, center=<0 -3 0>}]
 dyn_back_chain_1_jnt [{shapeType=Capsule, length=1.7289, radius=4.2711, shapeName=dyn_back_chain_1_jnt_capsule, center=<5.8373 -3.6382 0>}]
 dyn_back_chain_2_jnt [{shapeType=Capsule, length=1.7289, radius=4.2711, shapeName=dyn_back_chain_2_jnt_capsule, center=<4.3174 -3.4489 0>}]
 dyn_back_chain_3_jnt [{shapeType=Capsule, length=1.7289, radius=4.2711, shapeName=dyn_back_chain_3_jnt_capsule, center=<5.129 -1.2574 0>}]
 dyn_back_chain_4_jnt [{shapeType=Capsule, length=7.2289, radius=4.2711, shapeName=dyn_back_chain_4_jnt_capsule, center=<7.4962 -4.1489 0>, rotation=<rot 90 0 0>}]
 dyn_back_l_chain_1_jnt [{shapeType=Capsule, length=1.7289, radius=4.2711, shapeName=dyn_back_l_chain_1_jnt_capsule, center=<5.8373 -4.9711 0>}]
 dyn_back_l_chain_2_jnt [{shapeType=Capsule, length=1.7289, radius=4.2711, shapeName=dyn_back_l_chain_2_jnt_capsule, center=<3.7525 -4.5194 0>}]
 dyn_back_l_chain_3_jnt [{shapeType=Capsule, length=1.7289, radius=4.2711, shapeName=dyn_back_l_chain_3_jnt_capsule, center=<5.129 -1.6069 0>}]
 dyn_back_l_chain_4_jnt [{shapeType=Capsule, length=7.2289, radius=4.2711, shapeName=dyn_back_l_chain_4_jnt_capsule, center=<7.4962 -4.1489 0>, rotation=<rot 90 0 0>}]
 dyn_back_r_chain_1_jnt [{shapeType=Capsule, length=2.9789, radius=3.0211, shapeName=dyn_back_r_chain_1_jnt_capsule, center=<5.8373 -4.9711 0>}]
 dyn_back_r_chain_2_jnt [{shapeType=Capsule, length=2.2011, radius=3.7989, shapeName=dyn_back_r_chain_2_jnt_capsule, center=<4.1577 -4.0577 0>}]
 dyn_back_r_chain_3_jnt [{shapeType=Capsule, length=2.2011, radius=3.7989, shapeName=dyn_back_r_chain_3_jnt_capsule, center=<5.129 -2.5902 0>}]
 dyn_back_r_chain_4_jnt [{shapeType=Capsule, length=7.7011, radius=3.7989, shapeName=dyn_back_r_chain_4_jnt_capsule, center=<7.4962 -4.1489 0>, rotation=<rot 90 0 0>}]
 head [{shapeType=Capsule, length=6.9009, radius=6.5689, shapeName=head_capsule, rotation=<rot 90 0 0>}]
 master_L_chain_jnt [{shapeType=Capsule, length=0.505, radius=0.505, shapeName=master_L_chain_jnt_capsule, rotation=<rot 0 -90 0>}]
 master_R_chain_jnt [{shapeType=Capsule, length=3, radius=3, shapeName=master_R_chain_jnt_capsule, rotation=<rot 0 -90 0>}]
 master_hair_jnt [{shapeType=Capsule, length=0.505, radius=0.505, shapeName=master_hair_jnt_capsule, rotation=<rot 0 -90 0>}]
 neck_01 [{shapeType=Capsule, length=2.2217, radius=5.7986, shapeName=neck_01_capsule, rotation=<rot 90 0 0>}]
 neck_02 [{shapeType=Capsule, length=5.5914, radius=5.5215, shapeName=neck_02_capsule, rotation=<rot 90 0 0>}]
 spine_04 [
  {shapeType=Capsule, length=7.3487, radius=12.1826, shapeName=spine_04_capsule, center=<1.1566 0.5928 0>}
  {shapeType=Capsule, length=5.8809, radius=9.0313, shapeName=spine_04_capsule, center=<0.4636 6.6672 0>}
 ]
 spine_05 [{shapeType=Capsule, length=14.0638, radius=8.7878, shapeName=spine_05_capsule, center=<0 -2.5397 0>}]
 upperarm_l [{shapeType=Capsule, length=19.6713, radius=5.3287, shapeName=upperarm_l_capsule, center=<13.9129 0 0>, rotation=<rot 90 0 0>}]
 upperarm_r [{shapeType=Capsule, length=19.6736, radius=5.3264, shapeName=upperarm_r_capsule, center=<-14.8478 0 0.0065>, rotation=<rot 90 0 0>}]
}
constraints [
 {
  bone1Name=head
  bone2Name=neck_02
  swing1LimitDegrees=45
  swing1Motion=Free
  swing2LimitDegrees=45
  swing2Motion=Free
  twistLimitDegrees=45
  twistMotion=Free
 }
 {
  bone1Name=neck_02
  bone2Name=neck_01
  swing1LimitDegrees=45
  swing1Motion=Free
  swing2LimitDegrees=45
  swing2Motion=Free
  twistLimitDegrees=45
  twistMotion=Free
 }
 {
  bone1Name=neck_01
  bone2Name=spine_05
  swing1LimitDegrees=45
  swing1Motion=Free
  swing2LimitDegrees=45
  swing2Motion=Free
  twistLimitDegrees=45
  twistMotion=Free
 }
 {
  bone1Name=spine_05
  bone2Name=spine_04
  swing1LimitDegrees=45
  swing1Motion=Free
  swing2LimitDegrees=45
  swing2Motion=Free
  twistLimitDegrees=45
  twistMotion=Free
 }
 {
  bone1Name=clavicle_r
  bone2Name=spine_05
  swing1LimitDegrees=45
  swing1Motion=Free
  swing2LimitDegrees=45
  swing2Motion=Free
  twistLimitDegrees=45
  twistMotion=Free
 }
 {
  bone1Name=clavicle_l
  bone2Name=spine_05
  swing1LimitDegrees=45
  swing1Motion=Free
  swing2LimitDegrees=45
  swing2Motion=Free
  twistLimitDegrees=45
  twistMotion=Free
 }
 {
  bone1Name=upperarm_l
  bone2Name=clavicle_l
  swing1LimitDegrees=45
  swing1Motion=Free
  swing2LimitDegrees=45
  swing2Motion=Free
  twistLimitDegrees=45
  twistMotion=Free
 }
 {
  bone1Name=upperarm_r
  bone2Name=clavicle_r
  swing1LimitDegrees=45
  swing1Motion=Free
  swing2LimitDegrees=45
  swing2Motion=Free
  twistLimitDegrees=45
  twistMotion=Free
 }
 {bone1Name=dyn_back_chain_1_jnt, bone2Name=master_hair_jnt, swing1LimitDegrees=15, swing2LimitDegrees=15, twistLimitDegrees=1}
 {bone1Name=dyn_back_chain_2_jnt, bone2Name=dyn_back_chain_1_jnt, swing1LimitDegrees=15, swing2LimitDegrees=15, twistLimitDegrees=1}
 {bone1Name=dyn_back_chain_3_jnt, bone2Name=dyn_back_chain_2_jnt, swing1LimitDegrees=15, swing2LimitDegrees=15, twistLimitDegrees=1}
 {bone1Name=dyn_back_chain_4_jnt, bone2Name=dyn_back_chain_3_jnt, swing1LimitDegrees=15, swing2LimitDegrees=15, twistLimitDegrees=1}
 {bone1Name=dyn_back_l_chain_1_jnt, bone2Name=master_hair_jnt, swing1LimitDegrees=15, swing2LimitDegrees=5, twistLimitDegrees=1}
 {bone1Name=dyn_back_l_chain_2_jnt, bone2Name=dyn_back_l_chain_1_jnt, swing1LimitDegrees=15, swing2LimitDegrees=5, twistLimitDegrees=1}
 {bone1Name=dyn_back_l_chain_3_jnt, bone2Name=dyn_back_l_chain_2_jnt, swing1LimitDegrees=15, swing2LimitDegrees=5, twistLimitDegrees=1}
 {bone1Name=dyn_back_l_chain_4_jnt, bone2Name=dyn_back_l_chain_3_jnt, swing1LimitDegrees=15, swing2LimitDegrees=5, twistLimitDegrees=1}
 {bone1Name=dyn_back_r_chain_1_jnt, bone2Name=master_hair_jnt, swing1LimitDegrees=15, swing2LimitDegrees=5, twistLimitDegrees=1}
 {bone1Name=dyn_back_r_chain_2_jnt, bone2Name=dyn_back_r_chain_1_jnt, swing1LimitDegrees=15, swing2LimitDegrees=5, twistLimitDegrees=1}
 {bone1Name=dyn_back_r_chain_3_jnt, bone2Name=dyn_back_r_chain_2_jnt, swing1LimitDegrees=15, swing2LimitDegrees=5, twistLimitDegrees=1}
 {bone1Name=dyn_back_r_chain_4_jnt, bone2Name=dyn_back_r_chain_3_jnt, swing1LimitDegrees=15, swing2LimitDegrees=5, twistLimitDegrees=1}
 {bone1Name=dyn_R_chain_flip_jnt, bone2Name=dyn_R_chain_2_jnt, swing1LimitDegrees=15, swing2LimitDegrees=15, twistLimitDegrees=1}
 {bone1Name=dyn_R_chain_1_jnt, bone2Name=master_R_chain_jnt, swing1LimitDegrees=30, swing2LimitDegrees=30, twistLimitDegrees=1}
 {bone1Name=dyn_R_chain_2_jnt, bone2Name=dyn_R_chain_1_jnt, swing1LimitDegrees=30, swing2LimitDegrees=30, twistLimitDegrees=1}
 {bone1Name=dyn_R_chain_3_jnt, bone2Name=dyn_R_chain_2_jnt, swing1LimitDegrees=30, swing2LimitDegrees=30, twistLimitDegrees=1}
 {bone1Name=dyn_L_chain_3_jnt, bone2Name=dyn_L_chain_2_jnt, swing1LimitDegrees=35, swing2LimitDegrees=35, twistLimitDegrees=1}
 {bone1Name=dyn_L_chain_1_jnt, bone2Name=master_L_chain_jnt, swing1LimitDegrees=35, swing2LimitDegrees=35, twistLimitDegrees=1}
 {bone1Name=dyn_L_chain_2_jnt, bone2Name=dyn_L_chain_1_jnt, swing1LimitDegrees=35, swing2LimitDegrees=35, twistLimitDegrees=1}
 {bone1Name=dyn_R_side_hair_jnt, bone2Name=master_hair_jnt, swing1LimitDegrees=35, swing2LimitDegrees=35, twistLimitDegrees=1}
 {bone1Name=dyn_L_side_hair_jnt, bone2Name=master_hair_jnt, swing1LimitDegrees=35, swing2LimitDegrees=35, twistLimitDegrees=1}
 {
  bone1Name=master_hair_jnt
  bone2Name=head
  swing1LimitDegrees=45
  swing1Motion=Free
  swing2LimitDegrees=45
  swing2Motion=Free
  twistLimitDegrees=45
  twistMotion=Free
 }
 {
  bone1Name=master_L_chain_jnt
  bone2Name=master_hair_jnt
  swing1LimitDegrees=45
  swing1Motion=Free
  swing2LimitDegrees=45
  swing2Motion=Free
  twistLimitDegrees=45
  twistMotion=Free
 }
 {
  bone1Name=master_R_chain_jnt
  bone2Name=master_hair_jnt
  swing1LimitDegrees=45
  swing1Motion=Free
  swing2LimitDegrees=45
  swing2Motion=Free
  twistLimitDegrees=45
  twistMotion=Free
 }
]
