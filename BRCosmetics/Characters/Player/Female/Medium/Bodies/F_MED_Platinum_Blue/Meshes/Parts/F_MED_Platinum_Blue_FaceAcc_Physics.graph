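# Physics asset for the Platinum Blue female medium body's FaceAcc mesh, set up for dynamic hair on that cosmetic.
# Capsules on spine, neck, head, clavicles and upper arms form an upper-body proxy, mostly joined by free angular constraints.
# A master hair joint, constrained to spine_03, branches into bangs, side locks, left and right chains, and three back chains.
# Hair segments have swing and twist limits, with lower center and left back segments nearly twist-locked.
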PhysicsAsset /BRCosmetics/Characters/Player/Female/Medium/Bodies/F_MED_Platinum_Blue/Meshes/Parts/F_MED_Platinum_Blue_FaceAcc_Physics
bodies {
 clavicle_l [{shapeType=Capsule, length=4.8485, radius=5.625, shapeName=clavicle_l_capsule, center=<5.3443 0 -0.675>, rotation=<rot 90 0 0>}]
 clavicle_r [{shapeType=Capsule, length=14.3358, radius=6.875, shapeName=clavicle_l_capsule, center=<-7.0817 0 -0.1321>, rotation=<rot 90 0 0>}]
 dyn_L_chain_1_jnt [{shapeType=Capsule, length=4.3438, radius=3.9062, shapeName=dyn_L_chain_1_jnt_capsule, center=<3.4033 4 0>}]
 dyn_L_chain_2_jnt [{shapeType=Capsule, length=2.6484, radius=5.6016, shapeName=dyn_L_chain_2_jnt_capsule, center=<4.0218 2 0>}]
 dyn_L_chain_3_jnt [{shapeType=Capsule, length=4.8984, radius=4.8516, shapeName=dyn_L_chain_3_jnt_capsule, center=<4.3732 1.5 0>, rotation=<rot 90 0 0>}]
 dyn_L_side_hair_jnt [{
   shapeType=Capsule
   length=3.3438
   radius=2.9062
   shapeName=dyn_L_side_hair_jnt_capsule
   center=<2.5536 1.0626 0>
   rotation=<rot 90 180 179.9998>
  }]
 dyn_R_chain_1_jnt [{shapeType=Capsule, length=7.4237, radius=4.015, shapeName=dyn_R_chain_1_jnt_capsule, center=<2.325 5 0>}]
 dyn_R_chain_2_jnt [{shapeType=Capsule, length=11.4786, radius=5.3415, shapeName=dyn_R_chain_2_jnt_capsule, center=<3.8784 5 0>}]
 dyn_R_chain_3_jnt [{shapeType=Capsule, length=5.8858, radius=5.3642, shapeName=dyn_R_chain_3_jnt_capsule, center=<6.72 3 0>, rotation=<rot 90 0 0>}]
 dyn_R_chain_flip_jnt [{
   shapeType=Capsule
   length=2.25
   radius=5.6068
   shapeName=dyn_R_chain_flip_jnt_capsule
   center=<2.3903 1.8039 -1.38>
   rotation=<rot 60.0001 0.0004 0>
  }]
 dyn_R_side_hair_jnt [{
   shapeType=Capsule
   length=3.3438
   radius=2.9062
   shapeName=dyn_R_side_hair_jnt_capsule
   center=<2.5536 1.0626 0>
   rotation=<rot 90 180 179.9998>
  }]
 dyn_back_chain_1_jnt [{shapeType=Capsule, length=6.3438, radius=3.9062, shapeName=dyn_back_chain_1_jnt_capsule, center=<3.8444 -3 0>}]
 dyn_back_chain_2_jnt [{shapeType=Capsule, length=6.3438, radius=3.9062, shapeName=dyn_back_chain_2_jnt_capsule, center=<3.8444 -3 0>}]
 dyn_back_chain_3_jnt [{shapeType=Capsule, length=6.3438, radius=3.9062, shapeName=dyn_back_chain_3_jnt_capsule, center=<3.8444 -3 0>}]
 dyn_back_chain_4_jnt [{shapeType=Capsule, length=6.3438, radius=3.9062, shapeName=dyn_back_r_chain_4_jnt_capsule, center=<3.8444 -3 0>}]
 dyn_back_l_chain_1_jnt [{shapeType=Capsule, length=6.3438, radius=3.9062, shapeName=dyn_back_l_chain_1_jnt_capsule, center=<3.8444 -3 0>}]
 dyn_back_l_chain_2_jnt [{
   shapeType=Capsule
   length=6.3438
   radius=3.9062
   shapeName=dyn_back_l_chain_2_jnt_capsule
   center=<3.8444 -6.004 -2.9663>
   rotation=<rot 0 0 -44.6387>
  }]
 dyn_back_l_chain_3_jnt [{
   shapeType=Capsule
   length=6.3438
   radius=3.9062
   shapeName=dyn_back_l_chain_3_jnt_capsule
   center=<3.8444 -3.9578 -0.7387>
   rotation=<rot 0 0 -37.6427>
  }]
 dyn_back_l_chain_4_jnt [{shapeType=Capsule, length=6.3438, radius=3.9062, shapeName=dyn_back_r_chain_4_jnt_capsule, center=<3.8444 -3 0>}]
 dyn_back_r_chain_1_jnt [{shapeType=Capsule, length=4.7738, radius=5.4762, shapeName=dyn_back_r_chain_1_jnt_capsule, center=<3.6347 -2.4846 0>}]
 dyn_back_r_chain_2_jnt [{shapeType=Capsule, length=6.5047, radius=5.2453, shapeName=dyn_back_r_chain_2_jnt_capsule, center=<2.8495 -1.5947 0>}]
 dyn_back_r_chain_3_jnt [{shapeType=Capsule, length=8.0737, radius=4.1763, shapeName=dyn_back_r_chain_3_jnt_capsule, center=<4.2818 0 0>}]
 dyn_back_r_chain_4_jnt [{shapeType=Capsule, length=8.0737, radius=4.1763, shapeName=dyn_back_r_chain_4_jnt_capsule, center=<4.2818 0 0>}]
 dyn_bangs_jnt [{
   shapeType=Capsule
   length=6.3438
   radius=3.4062
   shapeName=dyn_bangs_jnt_capsule
   center=<1.1541 0.4805 -0.4148>
   rotation=<rot -90 -179.6835 179.6837>
  }]
 head [{shapeType=Capsule, length=6.9721, radius=8.2348, shapeName=head_capsule, rotation=<rot 90 0 0>}]
 master_L_chain_jnt [{shapeType=Capsule, length=1, radius=1, shapeName=master_L_chain_jnt_capsule}]
 master_R_chain_jnt [{shapeType=Capsule, length=1, radius=1, shapeName=master_R_chain_jnt_capsule}]
 master_back_chain_jnt [{shapeType=Capsule, length=0.505, radius=0.505, shapeName=master_back_chain_jnt_capsule, rotation=<rot 0 -90 0>}]
 master_hair_jnt [{shapeType=Capsule, length=1, radius=1, shapeName=master_hair_jnt_capsule}]
 neck_01 [{shapeType=Capsule, length=4.6965, radius=6.4367, shapeName=head_capsule, center=<2.5967 2.0576 0>, rotation=<rot 90 0 0>}]
 spine_03 [{shapeType=Capsule, length=4.5, radius=13.0139, shapeName=spine_04_capsule, center=<-0.3569 2.6072 0>}]
 spine_04 [{shapeType=Capsule, length=10.6139, radius=11.6361, shapeName=spine_04_capsule, center=<0.4833 0.86 0>}]
 spine_05 [
  {shapeType=Capsule, length=45.4114, radius=8.1927, shapeName=spine_05_capsule, center=<3.2244 -1.5257 0>}
  {shapeType=Capsule, length=59.1707, radius=8.0793, shapeName=spine_05_capsule, center=<6.3224 3.5738 0>}
 ]
 upperarm_l [{
   shapeType=Capsule
   length=24.5468
   radius=5.125
   shapeName=clavicle_l_capsule
   center=<10.4528 -1.6688 -0.5227>
   rotation=<rot 90 172.875 -167.1253>
  }]
 upperarm_r [{
   shapeType=Capsule
   length=20.4178
   radius=7.7106
   shapeName=clavicle_l_capsule
   center=<-11.3612 0.5412 -0.9313>
   rotation=<rot 90 0 19.8566>
  }]
}
constraints [
 {bone1Name=head, bone2Name=neck_01, swing1LimitDegrees=45, swing2LimitDegrees=45, twistLimitDegrees=45}
 {bone1Name=dyn_bangs_jnt, bone2Name=master_hair_jnt, swing1LimitDegrees=45, swing2LimitDegrees=45, twistLimitDegrees=45}
 {bone1Name=dyn_R_side_hair_jnt, bone2Name=master_hair_jnt, swing1LimitDegrees=45, swing2LimitDegrees=45, twistLimitDegrees=45}
 {bone1Name=dyn_L_side_hair_jnt, bone2Name=master_hair_jnt, swing1LimitDegrees=45, swing2LimitDegrees=45, twistLimitDegrees=45}
 {bone1Name=master_L_chain_jnt, bone2Name=master_hair_jnt, swing1LimitDegrees=45, swing2LimitDegrees=45, twistLimitDegrees=45}
 {bone1Name=dyn_L_chain_1_jnt, bone2Name=master_L_chain_jnt, swing1LimitDegrees=45, swing2LimitDegrees=20, twistLimitDegrees=20}
 {bone1Name=dyn_L_chain_2_jnt, bone2Name=dyn_L_chain_1_jnt, swing1LimitDegrees=45, swing2LimitDegrees=20, twistLimitDegrees=20}
 {bone1Name=dyn_L_chain_3_jnt, bone2Name=dyn_L_chain_2_jnt, swing1LimitDegrees=45, swing2LimitDegrees=20, twistLimitDegrees=20}
 {bone1Name=master_R_chain_jnt, bone2Name=master_hair_jnt, swing1LimitDegrees=45, swing2LimitDegrees=45, twistLimitDegrees=45}
 {bone1Name=dyn_R_chain_1_jnt, bone2Name=master_R_chain_jnt, swing1LimitDegrees=45, swing2LimitDegrees=45, twistLimitDegrees=45}
 {bone1Name=dyn_R_chain_2_jnt, bone2Name=dyn_R_chain_1_jnt, swing1LimitDegrees=45, swing2LimitDegrees=45, twistLimitDegrees=45}
 {bone1Name=dyn_R_chain_3_jnt, bone2Name=dyn_R_chain_2_jnt, swing1LimitDegrees=45, swing2LimitDegrees=45, twistLimitDegrees=45}
 {bone1Name=dyn_R_chain_flip_jnt, bone2Name=dyn_R_chain_2_jnt, swing1LimitDegrees=45, swing2LimitDegrees=45, twistLimitDegrees=45}
 {
  bone1Name=neck_01
  bone2Name=spine_05
  swing1LimitDegrees=45
  swing1Motion=Free
  swing2LimitDegrees=45
  swing2Motion=Free
  twistLimitDegrees=45
  twistMotion=Free
 }
 {
  bone1Name=spine_05
  bone2Name=spine_04
  swing1LimitDegrees=45
  swing1Motion=Free
  swing2LimitDegrees=45
  swing2Motion=Free
  twistLimitDegrees=45
  twistMotion=Free
 }
 {
  bone1Name=master_back_chain_jnt
  bone2Name=master_hair_jnt
  swing1LimitDegrees=45
  swing1Motion=Free
  swing2LimitDegrees=45
  swing2Motion=Free
  twistLimitDegrees=45
  twistMotion=Free
 }
 {bone1Name=dyn_back_chain_1_jnt, bone2Name=master_back_chain_jnt, swing1LimitDegrees=20, swing2LimitDegrees=20, twistLimitDegrees=20}
 {bone1Name=dyn_back_chain_2_jnt, bone2Name=dyn_back_chain_1_jnt, swing1LimitDegrees=25, swing2LimitDegrees=10, twistLimitDegrees=1}
 {bone1Name=dyn_back_chain_3_jnt, bone2Name=dyn_back_chain_2_jnt, swing1LimitDegrees=25, swing2LimitDegrees=10, twistLimitDegrees=1}
 {bone1Name=dyn_back_chain_4_jnt, bone2Name=dyn_back_chain_3_jnt, swing1LimitDegrees=25, swing2LimitDegrees=10, twistLimitDegrees=1}
 {bone1Name=dyn_back_r_chain_1_jnt, bone2Name=master_back_chain_jnt, swing1LimitDegrees=40, swing2LimitDegrees=40, twistLimitDegrees=40}
 {bone1Name=dyn_back_r_chain_2_jnt, bone2Name=dyn_back_r_chain_1_jnt, swing1LimitDegrees=40, swing2LimitDegrees=40, twistLimitDegrees=40}
 {bone1Name=dyn_back_l_chain_1_jnt, bone2Name=master_back_chain_jnt, swing1LimitDegrees=20, swing2LimitDegrees=20, twistLimitDegrees=20}
 {bone1Name=dyn_back_l_chain_2_jnt, bone2Name=dyn_back_l_chain_1_jnt, swing1LimitDegrees=25, swing2LimitDegrees=25, twistLimitDegrees=25}
 {bone1Name=dyn_back_l_chain_3_jnt, bone2Name=dyn_back_l_chain_2_jnt, swing1LimitDegrees=25, swing2LimitDegrees=10, twistLimitDegrees=1}
 {bone1Name=dyn_back_l_chain_4_jnt, bone2Name=dyn_back_l_chain_3_jnt, swing1LimitDegrees=25, swing2LimitDegrees=10, twistLimitDegrees=1}
 {
  bone1Name=spine_04
  bone2Name=spine_03
  swing1LimitDegrees=45
  swing1Motion=Free
  swing2LimitDegrees=45
  swing2Motion=Free
  twistLimitDegrees=45
  twistMotion=Free
 }
 {
  bone1Name=clavicle_l
  bone2Name=spine_03
  swing1LimitDegrees=45
  swing1Motion=Free
  swing2LimitDegrees=45
  swing2Motion=Free
  twistLimitDegrees=45
  twistMotion=Free
 }
 {
  bone1Name=upperarm_l
  bone2Name=spine_03
  swing1LimitDegrees=45
  swing1Motion=Free
  swing2LimitDegrees=45
  swing2Motion=Free
  twistLimitDegrees=45
  twistMotion=Free
 }
 {
  bone1Name=master_hair_jnt
  bone2Name=spine_03
  swing1LimitDegrees=45
  swing1Motion=Free
  swing2LimitDegrees=45
  swing2Motion=Free
  twistLimitDegrees=45
  twistMotion=Free
 }
 {
  bone1Name=clavicle_r
  bone2Name=spine_03
  swing1LimitDegrees=45
  swing1Motion=Free
  swing2LimitDegrees=45
  swing2Motion=Free
  twistLimitDegrees=45
  twistMotion=Free
 }
 {
  bone1Name=upperarm_r
  bone2Name=spine_03
  swing1LimitDegrees=45
  swing1Motion=Free
  swing2LimitDegrees=45
  swing2Motion=Free
  twistLimitDegrees=45
  twistMotion=Free
 }
 {bone1Name=dyn_back_r_chain_3_jnt, bone2Name=dyn_back_r_chain_2_jnt, swing1LimitDegrees=20, swing2LimitDegrees=20, twistLimitDegrees=20}
 {bone1Name=dyn_back_r_chain_4_jnt, bone2Name=dyn_back_r_chain_3_jnt, swing1LimitDegrees=20, swing2LimitDegrees=20, twistLimitDegrees=20}
]
