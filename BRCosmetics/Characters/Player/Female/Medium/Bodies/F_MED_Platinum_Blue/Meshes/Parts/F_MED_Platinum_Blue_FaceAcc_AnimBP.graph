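# Animation Blueprint for the Platinum Blue female face accessory, adding secondary motion to its bone chains over the cached PlatBlue pose.
# With Backbling or Cape set, a dedicated RigidBody setup and spine_05 constraints shape the back chains, and Cape also scales their master bone.
# Otherwise, bRigidBodyEnabled selects RigidBody simulation or an AnimDynamics fallback with spherical limits, constraints and bone offsets.
# A separate RigidBody branch uses a FrontEnd physics asset.
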AnimBlueprint /BRCosmetics/Characters/Player/Female/Medium/Bodies/F_MED_Platinum_Blue/Meshes/Parts/F_MED_Platinum_Blue_FaceAcc_AnimBP
target_skeleton=&/F_MED_Platinum_Blue_FaceAcc_Skeleton
graphs [
 AnimGraph {nodes [
   Anim Dynamics - Bone: dyn_back_l_chain_1_jnt(Alpha=1, ExternalForce="0,0,0") {
    angularDampingOverride=0.99
    angularSpringConstant=100
    bAngularSpring=true
    bGravityOverrideInSimSpace=true
    bOverrideAngularDamping=true
    bUseGravityOverride=true
    simulationSpace=BoneRelative
    alphaBoolBlend {blendOption=Linear}
    boundBone {boneName=dyn_back_l_chain_1_jnt}
    componentAppliedLinearAccClamp=<3000 3000 3000>
    componentLinearAccScale=<0.7 0.7 0.7>
    componentLinearVelScale=<0.7 0.7 0.7>
    gravityOverride {z=-980}
    physicsBodyDefinitions [{
      sphereCollisionRadius=0
      boundBone {boneName=dyn_back_l_chain_1_jnt}
      boxExtents=<20 10 10>
      constraintSetup {angularLimitsMax=<1 10 2>, angularLimitsMin=<-1 -10 -15>, angularTarget=<0 0 0>}
      localJointOffset=<3 0 0>
     }]
    relativeSpaceBone {boneName=root}
   }
   Anim Dynamics - Bone: dyn_back_r_chain_1_jnt(Alpha=1, ExternalForce="0,0,0") {
    angularDampingOverride=0.99
    angularSpringConstant=100
    bAngularSpring=true
    bGravityOverrideInSimSpace=true
    bOverrideAngularDamping=true
    bUseGravityOverride=true
    simulationSpace=BoneRelative
    alphaBoolBlend {blendOption=Linear}
    boundBone {boneName=dyn_back_r_chain_1_jnt}
    componentAppliedLinearAccClamp=<3000 3000 3000>
    componentLinearAccScale=<0.7 0.7 0.7>
    componentLinearVelScale=<0.7 0.7 0.7>
    gravityOverride {z=-980}
    physicsBodyDefinitions [{
      sphereCollisionRadius=0
      boundBone {boneName=dyn_back_r_chain_1_jnt}
      boxExtents=<20 10 10>
      constraintSetup {angularLimitsMax=<1 10 2>, angularLimitsMin=<-1 -10 -15>, angularTarget=<0 0 0>}
      localJointOffset=<3 0 0>
     }]
    relativeSpaceBone {boneName=root}
   }
   "Transform (Modify) Bone - Bone: dyn_back_l_chain_1_jnt"!(ComponentPose, Alpha=1, Translation="0,0,0", Rotation="0.814,13.201,-42.441") -> (Pose) {
    rotationMode=BMM_Additive
    rotationSpace=BCS_BoneSpace
    translationMode=BMM_Additive
    translationSpace=BCS_BoneSpace
    alphaBoolBlend {blendOption=Linear}
    alphaScaleBiasClamp {bInterpResult=true, interpSpeedDecreasing=5, interpSpeedIncreasing=7}
    boneToModify {boneName=dyn_back_l_chain_1_jnt}
    rotation=<rot 0.8141 13.2011 -42.441>
   }
   "Transform (Modify) Bone - Bone: dyn_back_r_chain_1_jnt"!(ComponentPose, Alpha=1, Translation="0,0,0", Rotation="30.772,-19.313,-66.307") -> (Pose) {
    rotationMode=BMM_Additive
    rotationSpace=BCS_BoneSpace
    translationMode=BMM_Additive
    translationSpace=BCS_BoneSpace
    alphaBoolBlend {blendOption=Linear}
    alphaScaleBiasClamp {bInterpResult=true, interpSpeedDecreasing=5, interpSpeedIncreasing=7}
    boneToModify {boneName=dyn_back_r_chain_1_jnt}
    rotation=<rot 30.772 -19.313 -66.307>
   }
   "Transform (Modify) Bone - Bone: master_back_chain_jnt"!(ComponentPose, Alpha=1, Translation="0,0,0", Rotation="0,-8,0") -> (Pose) {
    rotationMode=BMM_Additive
    rotationSpace=BCS_BoneSpace
    translationMode=BMM_Additive
    translationSpace=BCS_BoneSpace
    alphaBoolBlend {blendOption=Linear}
    alphaScaleBiasClamp {bInterpResult=true, interpSpeedDecreasing=5, interpSpeedIncreasing=7}
    boneToModify {boneName=master_back_chain_jnt}
    rotation {yaw=-8}
   }
   n1 = "Constraint - master_back_chain_jnt (P)"(ComponentPose=Local To Component(LocalPose=Use cached pose 'PlatBlue'()), Alpha=1, ConstraintWeights_0=1) {
    alphaBoolBlend {blendOption=Linear}
    boneToModify {boneName=master_back_chain_jnt}
    constraintSetup [{targetBone {boneName=head}}]
    constraintWeights [1]
   }
   n2 = "Constraint - master_back_chain_jnt (R)"(ComponentPose=@n1, Alpha=1, ConstraintWeights_0=1) {
    alphaBoolBlend {blendOption=Linear}
    boneToModify {boneName=master_back_chain_jnt}
    constraintSetup [{transformType=Rotation, targetBone {boneName=spine_05}}]
    constraintWeights [1]
   }
   n3 = RigidBody!(ComponentPose=@n2, Alpha, OverrideWorldGravity:Vector="0,0,-980", ExternalForce) -> (Pose) {
    bClampLinearTranslationLimitToRefPose=true
    bOverrideWorldGravity=true
    overridePhysicsAsset=&/F_MED_Platinum_Blue_FaceAcc_Physics
    simulationSpace=BaseBoneSpace
    alphaBoolBlend {blendOption=Linear}
    baseBoneRef {boneName=spine_04}
    componentAppliedLinearAccClamp=<3000 3000 300>
    componentLinearAccScale=<0.4 0.4 0.45>
    componentLinearVelScale=<0.35 0.3 0.2>
    overrideWorldGravity {z=-980}
    simSpaceSettings {
     maxAngularAcceleration=1000
     maxAngularVelocity=3
     maxLinearAcceleration=700
     maxLinearVelocity=500
     velocityScaleZ=0
     worldAlpha=1
     externalLinearDragV=<0.2 0.2>
    }
   }
   n4 = "Constraint - dyn_back_r_chain_2_jnt (PP)"!(ComponentPose=@n3, Alpha=1, ConstraintWeights_0=0.3, ConstraintWeights_1:float=0.7) -> (Pose) {
    alphaBoolBlend {blendOption=Linear}
    alphaScaleBiasClamp {bClampResult=true, bInterpResult=true, clampMax=0.8, interpSpeedDecreasing=5, interpSpeedIncreasing=7}
    boneToModify {boneName=dyn_back_r_chain_2_jnt}
    constraintSetup [{targetBone {boneName=clavicle_r}}, {targetBone {boneName=neck_01}}]
    constraintWeights [1,1]
   }
   n5 = "Constraint - dyn_back_l_chain_2_jnt (PP)"!(ComponentPose=@n4, Alpha=1, ConstraintWeights_0=0.3, ConstraintWeights_1:float=0.7) -> (Pose) {
    alphaBoolBlend {blendOption=Linear}
    alphaScaleBiasClamp {bClampResult=true, bInterpResult=true, clampMax=0.8, interpSpeedDecreasing=5, interpSpeedIncreasing=7}
    boneToModify {boneName=dyn_back_l_chain_2_jnt}
    constraintSetup [{targetBone {boneName=clavicle_l}}, {targetBone {boneName=neck_01}}]
    constraintWeights [1,1]
   }
   n6 = "Constraint - dyn_back_r_chain_3_jnt (PP)"!(ComponentPose=@n5, Alpha=1, ConstraintWeights_0=0.3, ConstraintWeights_1:float=0.7) -> (Pose) {
    alphaBoolBlend {blendOption=Linear}
    alphaScaleBiasClamp {bClampResult=true, bInterpResult=true, interpSpeedDecreasing=12, interpSpeedIncreasing=15}
    boneToModify {boneName=dyn_back_r_chain_3_jnt}
    constraintSetup [{targetBone {boneName=spine_05}}, {targetBone {boneName=neck_01}}]
    constraintWeights [1,1]
   }
   n7 = "Constraint - dyn_back_l_chain_3_jnt (PP)"!(ComponentPose=@n6, Alpha=1, ConstraintWeights_0=0.3, ConstraintWeights_1:float=0.7) -> (Pose) {
    alphaBoolBlend {blendOption=Linear}
    alphaScaleBiasClamp {bClampResult=true, bInterpResult=true, interpSpeedDecreasing=12, interpSpeedIncreasing=15}
    boneToModify {boneName=dyn_back_l_chain_3_jnt}
    constraintSetup [{targetBone {boneName=spine_05}}, {targetBone {boneName=neck_01}}]
    constraintWeights [1,1]
   }
   n8 = "Constraint - dyn_back_l_chain_2_jnt (P)"(ComponentPose=Local To Component(LocalPose=Use cached pose 'PlatBlue'()), Alpha=0.2, ConstraintWeights_0=1) {
    alphaBoolBlend {blendOption=Linear}
    boneToModify {boneName=dyn_back_l_chain_2_jnt}
    constraintSetup [{targetBone {boneName=neck_01}}]
    constraintWeights [1]
   }
   n9 = "Constraint - dyn_back_chain_2_jnt (P)"(ComponentPose=@n8, Alpha=0.5, ConstraintWeights_0=1) {
    alphaBoolBlend {blendOption=Linear}
    boneToModify {boneName=dyn_back_chain_2_jnt}
    constraintSetup [{targetBone {boneName=neck_01}}]
    constraintWeights [1]
   }
   n10 = "Constraint - dyn_back_r_chain_2_jnt (P)"(ComponentPose=@n9, Alpha=0.2, ConstraintWeights_0=1) {
    alphaBoolBlend {blendOption=Linear}
    boneToModify {boneName=dyn_back_r_chain_2_jnt}
    constraintSetup [{targetBone {boneName=neck_01}}]
    constraintWeights [1]
   }
   n11 = "Constraint - dyn_back_l_chain_3_jnt (PP)"(ComponentPose=@n10, Alpha=0.8, ConstraintWeights_0=0.5, ConstraintWeights_1=0.5) {
    alphaBoolBlend {blendOption=Linear}
    boneToModify {boneName=dyn_back_l_chain_3_jnt}
    constraintSetup [{targetBone {boneName=spine_05}}, {targetBone {boneName=clavicle_l}}]
    constraintWeights [1,1]
   }
   n12 = "Constraint - dyn_back_chain_3_jnt (P)"(ComponentPose=@n11, Alpha=0.3, ConstraintWeights_0=1) {
    alphaBoolBlend {blendOption=Linear}
    boneToModify {boneName=dyn_back_chain_3_jnt}
    constraintSetup [{targetBone {boneName=spine_05}}]
    constraintWeights [1]
   }
   n13 = "Constraint - dyn_back_r_chain_3_jnt (PP)"(ComponentPose=@n12, Alpha=0.8, ConstraintWeights_0=0.5, ConstraintWeights_1=0.5) {
    alphaBoolBlend {blendOption=Linear}
    boneToModify {boneName=dyn_back_r_chain_3_jnt}
    constraintSetup [{targetBone {boneName=spine_05}}, {targetBone {boneName=clavicle_r}}]
    constraintWeights [1,1]
   }
   n14 = "Constraint - dyn_back_l_chain_4_jnt (PP)"!(ComponentPose=@n13, Alpha=1, ConstraintWeights_0=0.5, ConstraintWeights_1:float=0.5) -> (Pose) {
    alphaBoolBlend {blendOption=Linear}
    boneToModify {boneName=dyn_back_l_chain_4_jnt}
    constraintSetup [{targetBone {boneName=spine_05}}, {targetBone {boneName=clavicle_l}}]
    constraintWeights [1,1]
   }
   n15 = "Constraint - dyn_back_chain_4_jnt (P)"(ComponentPose=@n14, Alpha=0.2, ConstraintWeights_0=1) {
    alphaBoolBlend {blendOption=Linear}
    boneToModify {boneName=dyn_back_chain_4_jnt}
    constraintSetup [{targetBone {boneName=spine_05}}]
    constraintWeights [1]
   }
   n16 = "Constraint - dyn_back_r_chain_4_jnt (PP)"!(ComponentPose=@n15, Alpha=1, ConstraintWeights_0=0.5, ConstraintWeights_1:float=0.5) -> (Pose) {
    alphaBoolBlend {blendOption=Linear}
    boneToModify {boneName=dyn_back_r_chain_4_jnt}
    constraintSetup [{targetBone {boneName=spine_05}}, {targetBone {boneName=clavicle_r}}]
    constraintWeights [1,1]
   }
   n17 = "Constraint - dyn_L_chain_2_jnt (PP)"!(ComponentPose=@n16, Alpha=1, ConstraintWeights_0=0.5, ConstraintWeights_1:float=0.5) -> (Pose) {
    alphaBoolBlend {blendOption=Linear}
    boneToModify {boneName=dyn_L_chain_2_jnt}
    constraintSetup [{targetBone {boneName=spine_05}}, {targetBone {boneName=clavicle_l}}]
    constraintWeights [1,1]
   }
   n18 = "Constraint - dyn_R_chain_2_jnt (PP)"!(ComponentPose=@n17, Alpha=1, ConstraintWeights_0=0.5, ConstraintWeights_1:float=0.5) -> (Pose) {
    alphaBoolBlend {blendOption=Linear}
    boneToModify {boneName=dyn_R_chain_2_jnt}
    constraintSetup [{targetBone {boneName=spine_05}}, {targetBone {boneName=clavicle_r}}]
    constraintWeights [1,1]
   }
   n19 = Anim Dynamics - Bone: dyn_L_chain_1_jnt!(ComponentPose=@n18, Alpha=1, GravityOverride:Vector="0,0,-980", ExternalForce="0,0,0") -> (Pose) {
    angularDampingOverride=0.9999
    angularSpringConstant=500
    bAngularSpring=true
    bGravityOverrideInSimSpace=true
    bLinearSpring=true
    bOverrideAngularDamping=true
    bUseGravityOverride=true
    linearSpringConstant=50
    simulationSpace=BoneRelative
    alphaBoolBlend {blendOption=Linear}
    boundBone {boneName=dyn_L_chain_1_jnt}
    componentAppliedLinearAccClamp=<1500 1500 0>
    componentLinearAccScale=<0.3 0.3 0.3>
    componentLinearVelScale=<0.3 0.3 0.3>
    gravityOverride {z=-980}
    physicsBodyDefinitions [{
      sphereCollisionRadius=0
      boundBone {boneName=dyn_L_chain_1_jnt}
      boxExtents=<5 20 20>
      constraintSetup {angularLimitsMax=<1 5 5>, angularLimitsMin=<-1 -5 -2>, angularTarget=<0 0 0>}
      localJointOffset=<3 0 0>
     }]
    relativeSpaceBone {boneName=root}
   }
   n20 = Anim Dynamics - Bone: dyn_L_chain_2_jnt!(ComponentPose=@n19, Alpha=1, GravityOverride:Vector="0,0,-980", ExternalForce="0,0,0") -> (Pose) {
    angularDampingOverride=0.9999
    angularSpringConstant=500
    bAngularSpring=true
    bGravityOverrideInSimSpace=true
    bLinearSpring=true
    bOverrideAngularDamping=true
    bUseGravityOverride=true
    linearSpringConstant=50
    simulationSpace=BoneRelative
    alphaBoolBlend {blendOption=Linear}
    boundBone {boneName=dyn_L_chain_2_jnt}
    componentAppliedLinearAccClamp=<1500 1500 0>
    componentLinearAccScale=<0.3 0.3 0.3>
    componentLinearVelScale=<0.3 0.3 0.3>
    gravityOverride {z=-980}
    physicsBodyDefinitions [{
      sphereCollisionRadius=0
      boundBone {boneName=dyn_L_chain_2_jnt}
      boxExtents=<5 20 20>
      constraintSetup {angularLimitsMax=<1 5 5>, angularLimitsMin=<-1 -5 -2>, angularTarget=<0 0 0>}
      localJointOffset=<3 0 0>
     }]
    relativeSpaceBone {boneName=root}
   }
   n21 = Anim Dynamics - Bone: dyn_L_chain_3_jnt!(ComponentPose=@n20, Alpha=1, GravityOverride:Vector="0,0,-980", ExternalForce="0,0,0") -> (Pose) {
    angularDampingOverride=0.9999
    angularSpringConstant=500
    bAngularSpring=true
    bGravityOverrideInSimSpace=true
    bLinearSpring=true
    bOverrideAngularDamping=true
    bUseGravityOverride=true
    linearSpringConstant=50
    simulationSpace=BoneRelative
    alphaBoolBlend {blendOption=Linear}
    boundBone {boneName=dyn_L_chain_3_jnt}
    componentAppliedLinearAccClamp=<1500 1500 0>
    componentLinearAccScale=<0.3 0.3 0.3>
    componentLinearVelScale=<0.3 0.3 0.3>
    gravityOverride {z=-980}
    physicsBodyDefinitions [{
      sphereCollisionRadius=0
      boundBone {boneName=dyn_L_chain_3_jnt}
      boxExtents=<5 20 20>
      constraintSetup {angularLimitsMax=<1 5 5>, angularLimitsMin=<-1 -5 -2>, angularTarget=<0 0 0>}
      localJointOffset=<3 0 0>
     }]
    relativeSpaceBone {boneName=root}
   }
   n22 = Anim Dynamics - Bone: dyn_R_chain_1_jnt!(ComponentPose=@n21, Alpha=1, GravityOverride:Vector="0,0,-980", ExternalForce="0,0,0") -> (Pose) {
    angularDampingOverride=0.9999
    angularSpringConstant=100
    bAngularSpring=true
    bGravityOverrideInSimSpace=true
    bOverrideAngularDamping=true
    bUseGravityOverride=true
    linearSpringConstant=10
    simulationSpace=BoneRelative
    alphaBoolBlend {blendOption=Linear}
    boundBone {boneName=dyn_R_chain_1_jnt}
    componentAppliedLinearAccClamp=<3500 3500 3500>
    componentLinearAccScale=<0.2 0.2 0.2>
    componentLinearVelScale=<0.2 0.2 0.2>
    gravityOverride {z=-980}
    physicsBodyDefinitions [{
      sphereCollisionRadius=0
      boundBone {boneName=dyn_R_chain_1_jnt}
      boxExtents=<5 25 25>
      constraintSetup {angularLimitsMax=<1 5 5>, angularLimitsMin=<-1 -5 -2>, angularTarget=<0 0 0>}
      localJointOffset=<3 0 0>
     }]
    relativeSpaceBone {boneName=root}
   }
   n23 = Anim Dynamics - Bone: dyn_R_chain_2_jnt!(ComponentPose=@n22, Alpha=1, GravityOverride:Vector="0,0,-980", ExternalForce="0,0,0") -> (Pose) {
    angularDampingOverride=0.9999
    angularSpringConstant=100
    bAngularSpring=true
    bGravityOverrideInSimSpace=true
    bOverrideAngularDamping=true
    bUseGravityOverride=true
    linearSpringConstant=10
    simulationSpace=BoneRelative
    alphaBoolBlend {blendOption=Linear}
    boundBone {boneName=dyn_R_chain_2_jnt}
    componentAppliedLinearAccClamp=<3500 3500 3500>
    componentLinearAccScale=<0.2 0.2 0.2>
    componentLinearVelScale=<0.2 0.2 0.2>
    gravityOverride {z=-980}
    physicsBodyDefinitions [{
      sphereCollisionRadius=0
      boundBone {boneName=dyn_R_chain_2_jnt}
      boxExtents=<5 25 25>
      constraintSetup {angularLimitsMax=<1 5 5>, angularLimitsMin=<-1 -5 -2>, angularTarget=<0 0 0>}
      localJointOffset=<3 0 0>
     }]
    relativeSpaceBone {boneName=root}
   }
   n24 = Anim Dynamics - Bone: dyn_R_chain_3_jnt!(ComponentPose=@n23, Alpha=1, GravityOverride:Vector="0,0,-980", ExternalForce="0,0,0") -> (Pose) {
    angularDampingOverride=0.9999
    angularSpringConstant=100
    bAngularSpring=true
    bGravityOverrideInSimSpace=true
    bOverrideAngularDamping=true
    bUseGravityOverride=true
    linearSpringConstant=10
    simulationSpace=BoneRelative
    alphaBoolBlend {blendOption=Linear}
    boundBone {boneName=dyn_R_chain_3_jnt}
    componentAppliedLinearAccClamp=<3500 3500 3500>
    componentLinearAccScale=<0.2 0.2 0.2>
    componentLinearVelScale=<0.2 0.2 0.2>
    gravityOverride {z=-980}
    physicsBodyDefinitions [{
      sphereCollisionRadius=0
      boundBone {boneName=dyn_R_chain_3_jnt}
      boxExtents=<5 25 25>
      constraintSetup {angularLimitsMax=<1 5 5>, angularLimitsMin=<-1 -5 -2>, angularTarget=<0 0 0>}
      localJointOffset=<3 0 0>
     }]
    relativeSpaceBone {boneName=root}
   }
   n25 = Anim Dynamics - Bone: dyn_back_l_chain_2_jnt!(ComponentPose=@n24, Alpha=1, GravityOverride:Vector="0,0,-980", ExternalForce="0,0,0") -> (Pose) {
    angularDampingOverride=0.99
    angularSpringConstant=100
    bAngularSpring=true
    bGravityOverrideInSimSpace=true
    bOverrideAngularDamping=true
    bUseGravityOverride=true
    simulationSpace=BoneRelative
    alphaBoolBlend {blendOption=Linear}
    boundBone {boneName=dyn_back_l_chain_2_jnt}
    componentAppliedLinearAccClamp=<3000 3000 3000>
    componentLinearAccScale=<0.7 0.7 0.7>
    componentLinearVelScale=<0.7 0.7 0.7>
    gravityOverride {z=-980}
    physicsBodyDefinitions [{
      sphereCollisionRadius=0
      boundBone {boneName=dyn_back_l_chain_2_jnt}
      boxExtents=<20 10 10>
      constraintSetup {angularLimitsMax=<1 10 2>, angularLimitsMin=<-1 -5 -15>, angularTarget=<0 0 0>}
      localJointOffset=<3 0 0>
     }]
    relativeSpaceBone {boneName=root}
   }
   n26 = Anim Dynamics - Bone: dyn_back_l_chain_3_jnt!(ComponentPose=@n25, Alpha=1, GravityOverride:Vector="0,0,-980", ExternalForce="0,0,0") -> (Pose) {
    angularDampingOverride=0.99
    angularSpringConstant=100
    bAngularSpring=true
    bGravityOverrideInSimSpace=true
    bOverrideAngularDamping=true
    bUseGravityOverride=true
    bUseSphericalLimits=true
    simulationSpace=BoneRelative
    alphaBoolBlend {blendOption=Linear}
    boundBone {boneName=dyn_back_l_chain_3_jnt}
    componentAppliedLinearAccClamp=<3000 3000 3000>
    componentLinearAccScale=<0.7 0.7 0.7>
    componentLinearVelScale=<0.7 0.7 0.7>
    gravityOverride {z=-980}
    physicsBodyDefinitions [{
      collisionType=CustomSphere
      sphereCollisionRadius=2
      boundBone {boneName=dyn_back_l_chain_3_jnt}
      boxExtents=<20 10 10>
      constraintSetup {angularLimitsMax=<1 10 2>, angularLimitsMin=<-1 -5 -15>, angularTarget=<0 0 0>}
      localJointOffset=<3 0 0>
     }]
    relativeSpaceBone {boneName=root}
    sphericalLimits [
     {limitRadius=5, drivingBone {boneName=spine_05}, sphereLocalOffset=<0 0 0>}
     {limitRadius=5, drivingBone {boneName=clavicle_l}, sphereLocalOffset=<0 0 0>}
    ]
   }
   n27 = Anim Dynamics - Bone: dyn_back_l_chain_4_jnt!(ComponentPose=@n26, Alpha=1, GravityOverride:Vector="0,0,-980", ExternalForce="0,0,0") -> (Pose) {
    angularDampingOverride=0.99
    angularSpringConstant=100
    bAngularSpring=true
    bGravityOverrideInSimSpace=true
    bOverrideAngularDamping=true
    bUseGravityOverride=true
    bUseSphericalLimits=true
    simulationSpace=BoneRelative
    alphaBoolBlend {blendOption=Linear}
    boundBone {boneName=dyn_back_l_chain_4_jnt}
    componentAppliedLinearAccClamp=<3000 3000 3000>
    componentLinearAccScale=<0.7 0.7 0.7>
    componentLinearVelScale=<0.7 0.7 0.7>
    gravityOverride {z=-980}
    physicsBodyDefinitions [{
      collisionType=CustomSphere
      sphereCollisionRadius=2
      boundBone {boneName=dyn_back_l_chain_4_jnt}
      boxExtents=<20 10 10>
      constraintSetup {angularLimitsMax=<1 10 2>, angularLimitsMin=<-1 -5 -15>, angularTarget=<0 0 0>}
      localJointOffset=<3 0 0>
     }]
    relativeSpaceBone {boneName=root}
    sphericalLimits [
     {limitRadius=6, drivingBone {boneName=spine_05}, sphereLocalOffset=<0 0 0>}
     {limitRadius=5, drivingBone {boneName=clavicle_l}, sphereLocalOffset=<0 0 0>}
    ]
   }
   n28 = Anim Dynamics - Bone: dyn_back_chain_1_jnt!(ComponentPose=@n27, Alpha=1, GravityOverride:Vector="0,0,-980", ExternalForce="0,0,0") -> (Pose) {
    angularDampingOverride=0.99
    angularSpringConstant=100
    bAngularSpring=true
    bGravityOverrideInSimSpace=true
    bOverrideAngularDamping=true
    bUseGravityOverride=true
    simulationSpace=BoneRelative
    alphaBoolBlend {blendOption=Linear}
    boundBone {boneName=dyn_back_chain_1_jnt}
    componentAppliedLinearAccClamp=<3000 3000 3000>
    componentLinearAccScale=<0.7 0.7 0.7>
    componentLinearVelScale=<0.7 0.7 0.7>
    gravityOverride {z=-980}
    physicsBodyDefinitions [{
      sphereCollisionRadius=0
      boundBone {boneName=dyn_back_chain_1_jnt}
      boxExtents=<20 10 10>
      constraintSetup {angularLimitsMax=<1 10 2>, angularLimitsMin=<-1 -10 -20>, angularTarget=<0 0 0>}
      localJointOffset=<3 0 0>
     }]
    relativeSpaceBone {boneName=root}
   }
   n29 = Anim Dynamics - Bone: dyn_back_chain_2_jnt!(ComponentPose=@n28, Alpha=1, GravityOverride:Vector="0,0,-980", ExternalForce="0,0,0") -> (Pose) {
    angularDampingOverride=0.99
    angularSpringConstant=100
    bAngularSpring=true
    bGravityOverrideInSimSpace=true
    bOverrideAngularDamping=true
    bUseGravityOverride=true
    simulationSpace=BoneRelative
    alphaBoolBlend {blendOption=Linear}
    boundBone {boneName=dyn_back_chain_2_jnt}
    componentAppliedLinearAccClamp=<3000 3000 3000>
    componentLinearAccScale=<0.7 0.7 0.7>
    componentLinearVelScale=<0.7 0.7 0.7>
    gravityOverride {z=-980}
    physicsBodyDefinitions [{
      sphereCollisionRadius=0
      boundBone {boneName=dyn_back_chain_2_jnt}
      boxExtents=<20 10 10>
      constraintSetup {angularLimitsMax=<1 10 2>, angularLimitsMin=<-1 -10 -20>, angularTarget=<0 0 0>}
      localJointOffset=<3 0 0>
     }]
    relativeSpaceBone {boneName=root}
   }
   n30 = Anim Dynamics - Bone: dyn_back_chain_3_jnt!(ComponentPose=@n29, Alpha=1, GravityOverride:Vector="0,0,-980", ExternalForce="0,0,0") -> (Pose) {
    angularDampingOverride=0.99
    angularSpringConstant=100
    bAngularSpring=true
    bGravityOverrideInSimSpace=true
    bOverrideAngularDamping=true
    bUseGravityOverride=true
    bUseSphericalLimits=true
    simulationSpace=BoneRelative
    alphaBoolBlend {blendOption=Linear}
    boundBone {boneName=dyn_back_chain_3_jnt}
    componentAppliedLinearAccClamp=<3000 3000 3000>
    componentLinearAccScale=<0.7 0.7 0.7>
    componentLinearVelScale=<0.7 0.7 0.7>
    gravityOverride {z=-980}
    physicsBodyDefinitions [{
      collisionType=CustomSphere
      sphereCollisionRadius=2
      boundBone {boneName=dyn_back_chain_3_jnt}
      boxExtents=<20 10 10>
      constraintSetup {angularLimitsMax=<1 10 2>, angularLimitsMin=<-1 -10 -20>, angularTarget=<0 0 0>}
      localJointOffset=<3 0 0>
     }]
    relativeSpaceBone {boneName=root}
    sphericalLimits [{limitRadius=5, drivingBone {boneName=spine_05}, sphereLocalOffset=<0 0 0>}]
   }
   n31 = Anim Dynamics - Bone: dyn_back_chain_4_jnt!(ComponentPose=@n30, Alpha=1, GravityOverride:Vector="0,0,-980", ExternalForce="0,0,0") -> (Pose) {
    angularDampingOverride=0.99
    angularSpringConstant=100
    bAngularSpring=true
    bGravityOverrideInSimSpace=true
    bOverrideAngularDamping=true
    bUseGravityOverride=true
    bUseSphericalLimits=true
    simulationSpace=BoneRelative
    alphaBoolBlend {blendOption=Linear}
    boundBone {boneName=dyn_back_chain_4_jnt}
    componentAppliedLinearAccClamp=<3000 3000 3000>
    componentLinearAccScale=<0.7 0.7 0.7>
    componentLinearVelScale=<0.7 0.7 0.7>
    gravityOverride {z=-980}
    physicsBodyDefinitions [{
      collisionType=CustomSphere
      sphereCollisionRadius=2
      boundBone {boneName=dyn_back_chain_4_jnt}
      boxExtents=<20 10 10>
      constraintSetup {angularLimitsMax=<1 10 2>, angularLimitsMin=<-1 -10 -20>, angularTarget=<0 0 0>}
      localJointOffset=<3 0 0>
     }]
    relativeSpaceBone {boneName=root}
    sphericalLimits [{limitRadius=6, drivingBone {boneName=spine_05}, sphereLocalOffset=<0 0 0>}]
   }
   n32 = Anim Dynamics - Bone: dyn_back_r_chain_2_jnt!(ComponentPose=@n31, Alpha=1, GravityOverride:Vector="0,0,-980", ExternalForce="0,0,0") -> (Pose) {
    angularDampingOverride=0.99
    angularSpringConstant=100
    bAngularSpring=true
    bGravityOverrideInSimSpace=true
    bOverrideAngularDamping=true
    bUseGravityOverride=true
    bUsePlanarLimit=false
    bUseSphericalLimits=true
    simulationSpace=BoneRelative
    alphaBoolBlend {blendOption=Linear}
    boundBone {boneName=dyn_back_r_chain_2_jnt}
    componentAppliedLinearAccClamp=<3000 3000 3000>
    componentLinearAccScale=<0.7 0.7 0.7>
    componentLinearVelScale=<0.7 0.7 0.7>
    gravityOverride {z=-980}
    physicsBodyDefinitions [{
      collisionType=CustomSphere
      sphereCollisionRadius=2
      boundBone {boneName=dyn_back_r_chain_2_jnt}
      boxExtents=<20 10 10>
      constraintSetup {angularLimitsMax=<1 10 2>, angularLimitsMin=<-1 -5 -15>, angularTarget=<0 0 0>}
      localJointOffset=<3 0 0>
     }]
    relativeSpaceBone {boneName=root}
    sphericalLimits [
     {limitRadius=6, drivingBone {boneName=clavicle_r}, sphereLocalOffset=<0 0 0>}
     {limitRadius=6, drivingBone {boneName=upperarm_r}, sphereLocalOffset=<0 0 0>}
     {limitRadius=7, drivingBone {boneName=neck_01}, sphereLocalOffset=<0 0 0>}
    ]
   }
   n33 = Anim Dynamics - Bone: dyn_back_r_chain_3_jnt!(ComponentPose=@n32, Alpha=1, GravityOverride:Vector="0,0,-980", ExternalForce="0,0,0") -> (Pose) {
    angularDampingOverride=0.99
    angularSpringConstant=100
    bAngularSpring=true
    bGravityOverrideInSimSpace=true
    bOverrideAngularDamping=true
    bUseGravityOverride=true
    bUseSphericalLimits=true
    simulationSpace=BoneRelative
    alphaBoolBlend {blendOption=Linear}
    boundBone {boneName=dyn_back_r_chain_3_jnt}
    componentAppliedLinearAccClamp=<3000 3000 3000>
    componentLinearAccScale=<0.7 0.7 0.7>
    componentLinearVelScale=<0.7 0.7 0.7>
    gravityOverride {z=-980}
    physicsBodyDefinitions [{
      collisionType=CustomSphere
      sphereCollisionRadius=2
      boundBone {boneName=dyn_back_r_chain_3_jnt}
      boxExtents=<20 10 10>
      constraintSetup {angularLimitsMax=<1 10 2>, angularLimitsMin=<-1 -5 -15>, angularTarget=<0 0 0>}
      localJointOffset=<3 0 0>
     }]
    relativeSpaceBone {boneName=root}
    sphericalLimits [
     {limitRadius=5, drivingBone {boneName=spine_05}, sphereLocalOffset=<0 0 0>}
     {limitRadius=9, drivingBone {boneName=clavicle_r}, sphereLocalOffset=<0 0 0>}
     {limitRadius=9, drivingBone {boneName=upperarm_r}, sphereLocalOffset=<0 0 0>}
    ]
   }
   n34 = Anim Dynamics - Bone: dyn_back_r_chain_4_jnt!(ComponentPose=@n33, Alpha=1, GravityOverride:Vector="0,0,-980", ExternalForce="0,0,0") -> (Pose) {
    angularDampingOverride=0.99
    angularSpringConstant=100
    bAngularSpring=true
    bGravityOverrideInSimSpace=true
    bOverrideAngularDamping=true
    bUseGravityOverride=true
    bUseSphericalLimits=true
    simulationSpace=BoneRelative
    alphaBoolBlend {blendOption=Linear}
    boundBone {boneName=dyn_back_r_chain_4_jnt}
    componentAppliedLinearAccClamp=<3000 3000 3000>
    componentLinearAccScale=<0.7 0.7 0.7>
    componentLinearVelScale=<0.7 0.7 0.7>
    gravityOverride {z=-980}
    physicsBodyDefinitions [{
      collisionType=CustomSphere
      sphereCollisionRadius=2
      boundBone {boneName=dyn_back_r_chain_4_jnt}
      boxExtents=<20 10 10>
      constraintSetup {angularLimitsMax=<1 10 2>, angularLimitsMin=<-1 -5 -15>, angularTarget=<0 0 0>}
      localJointOffset=<3 0 0>
     }]
    relativeSpaceBone {boneName=root}
    sphericalLimits [
     {limitRadius=6, drivingBone {boneName=spine_05}, sphereLocalOffset=<0 0 0>}
     {limitRadius=9, drivingBone {boneName=clavicle_r}, sphereLocalOffset=<0 0 0>}
     {limitRadius=9, drivingBone {boneName=upperarm_r}, sphereLocalOffset=<0 0 0>}
    ]
   }
   n35 = "Transform (Modify) Bone - Bone: dyn_back_r_chain_1_jnt"(ComponentPose=@n34, Alpha=1, Translation="2,-5,0", Rotation="18.744,6.203,7.063") {
    rotationMode=BMM_Additive
    rotationSpace=BCS_BoneSpace
    translationMode=BMM_Additive
    translationSpace=BCS_BoneSpace
    alphaBoolBlend {blendOption=Linear}
    alphaScaleBiasClamp {bInterpResult=true, interpSpeedDecreasing=5, interpSpeedIncreasing=7}
    boneToModify {boneName=dyn_back_r_chain_1_jnt}
    rotation=<rot 18.744 6.203 7.063>
    translation=<2 -5>
   }
   n36 = "Transform (Modify) Bone - Bone: dyn_back_r_chain_1_jnt"(ComponentPose=@n35, Alpha=1, Translation="2,-5,0", Rotation="0,-5,0") {
    rotationMode=BMM_Additive
    rotationSpace=BCS_BoneSpace
    translationMode=BMM_Additive
    translationSpace=BCS_BoneSpace
    alphaBoolBlend {blendOption=Linear}
    alphaScaleBiasClamp {bInterpResult=true, interpSpeedDecreasing=5, interpSpeedIncreasing=7}
    boneToModify {boneName=dyn_back_r_chain_1_jnt}
    rotation {yaw=-5}
    translation=<2 -5>
   }
   n37 = "Constraint - dyn_back_l_chain_1_jnt (PP)"!(ComponentPose=@n36, Alpha=1, ConstraintWeights_0=0.5, ConstraintWeights_1:float=0.5) -> (Pose) {
    alphaBoolBlend {blendOption=Linear}
    alphaScaleBiasClamp {bInterpResult=true, interpSpeedDecreasing=5, interpSpeedIncreasing=7}
    boneToModify {boneName=dyn_back_l_chain_1_jnt}
    constraintSetup [{targetBone {boneName=head}}, {targetBone {boneName=neck_02}}]
    constraintWeights [1,1]
   }
   n38 = "Constraint - dyn_back_chain_1_jnt (PP)"!(ComponentPose=@n37, Alpha=1, ConstraintWeights_0=0.5, ConstraintWeights_1:float=0.5) -> (Pose) {
    alphaBoolBlend {blendOption=Linear}
    alphaScaleBiasClamp {bInterpResult=true, interpSpeedDecreasing=5, interpSpeedIncreasing=7}
    boneToModify {boneName=dyn_back_chain_1_jnt}
    constraintSetup [{targetBone {boneName=head}}, {targetBone {boneName=neck_02}}]
    constraintWeights [1,1]
   }
   n39 = "Constraint - dyn_back_r_chain_1_jnt (PP)"!(ComponentPose=@n38, Alpha=1, ConstraintWeights_0=0.5, ConstraintWeights_1:float=0.5) -> (Pose) {
    alphaBoolBlend {blendOption=Linear}
    alphaScaleBiasClamp {bInterpResult=true, interpSpeedDecreasing=5, interpSpeedIncreasing=7}
    boneToModify {boneName=dyn_back_r_chain_1_jnt}
    constraintSetup [{targetBone {boneName=head}}, {targetBone {boneName=neck_02}}]
    constraintWeights [1,1]
   }
   n40 = "Constraint - dyn_back_l_chain_2_jnt (PP)"(ComponentPose=@n39, Alpha=1, ConstraintWeights_0=0.5, ConstraintWeights_1=0.5) {
    alphaBoolBlend {blendOption=Linear}
    alphaScaleBiasClamp {bInterpResult=true, interpSpeedDecreasing=5, interpSpeedIncreasing=7}
    boneToModify {boneName=dyn_back_l_chain_2_jnt}
    constraintSetup [{targetBone {boneName=head}}, {targetBone {boneName=neck_02}}]
    constraintWeights [1,1]
   }
   n41 = "Constraint - dyn_back_chain_2_jnt (PP)"!(ComponentPose=@n40, Alpha=1, ConstraintWeights_0=0.5, ConstraintWeights_1:float=0.5) -> (Pose) {
    alphaBoolBlend {blendOption=Linear}
    alphaScaleBiasClamp {bInterpResult=true, interpSpeedDecreasing=5, interpSpeedIncreasing=7}
    boneToModify {boneName=dyn_back_chain_2_jnt}
    constraintSetup [{targetBone {boneName=head}}, {targetBone {boneName=neck_02}}]
    constraintWeights [1,1]
   }
   n42 = "Constraint - dyn_back_r_chain_2_jnt (PP)"(ComponentPose=@n41, Alpha=1, ConstraintWeights_0=0.5, ConstraintWeights_1=0.5) {
    alphaBoolBlend {blendOption=Linear}
    alphaScaleBiasClamp {bInterpResult=true, interpSpeedDecreasing=5, interpSpeedIncreasing=7}
    boneToModify {boneName=dyn_back_r_chain_2_jnt}
    constraintSetup [{targetBone {boneName=head}}, {targetBone {boneName=neck_02}}]
    constraintWeights [1,1]
   }
   n43 = "Constraint - dyn_back_l_chain_3_jnt (PP)"(ComponentPose=@n42, Alpha=1, ConstraintWeights_0=0.8, ConstraintWeights_1=0.2) {
    alphaBoolBlend {blendOption=Linear}
    alphaScaleBiasClamp {bInterpResult=true, interpSpeedDecreasing=5, interpSpeedIncreasing=7}
    boneToModify {boneName=dyn_back_l_chain_3_jnt}
    constraintSetup [{targetBone {boneName=spine_05}}, {targetBone {boneName=clavicle_l}}]
    constraintWeights [1,1]
   }
   n44 = "Constraint - dyn_back_chain_3_jnt (P)"(ComponentPose=@n43, Alpha=1, ConstraintWeights_0=1) {
    alphaBoolBlend {blendOption=Linear}
    alphaScaleBiasClamp {bInterpResult=true, interpSpeedDecreasing=5, interpSpeedIncreasing=7}
    boneToModify {boneName=dyn_back_chain_3_jnt}
    constraintSetup [{targetBone {boneName=spine_05}}]
    constraintWeights [1]
   }
   n45 = "Constraint - dyn_back_r_chain_3_jnt (PP)"(ComponentPose=@n44, Alpha=1, ConstraintWeights_0=0.8, ConstraintWeights_1=0.2) {
    alphaBoolBlend {blendOption=Linear}
    alphaScaleBiasClamp {bInterpResult=true, interpSpeedDecreasing=5, interpSpeedIncreasing=7}
    boneToModify {boneName=dyn_back_r_chain_3_jnt}
    constraintSetup [{targetBone {boneName=spine_05}}, {targetBone {boneName=clavicle_r}}]
    constraintWeights [1,1]
   }
   n46 = "Constraint - dyn_back_l_chain_4_jnt (PP)"(ComponentPose=@n45, Alpha=1, ConstraintWeights_0=0.8, ConstraintWeights_1=0.2) {
    alphaBoolBlend {blendOption=Linear}
    alphaScaleBiasClamp {bInterpResult=true, interpSpeedDecreasing=5, interpSpeedIncreasing=7}
    boneToModify {boneName=dyn_back_l_chain_4_jnt}
    constraintSetup [{targetBone {boneName=spine_04}}, {targetBone {boneName=clavicle_l}}]
    constraintWeights [1,1]
   }
   n47 = "Constraint - dyn_back_chain_4_jnt (P)"(ComponentPose=@n46, Alpha=0.1, ConstraintWeights_0=1) {
    alphaBoolBlend {blendOption=Linear}
    alphaScaleBiasClamp {bInterpResult=true, interpSpeedDecreasing=5, interpSpeedIncreasing=7}
    boneToModify {boneName=dyn_back_chain_4_jnt}
    constraintSetup [{targetBone {boneName=spine_05}}]
    constraintWeights [1]
   }
   n48 = "Constraint - dyn_back_r_chain_4_jnt (PP)"(ComponentPose=@n47, Alpha=1, ConstraintWeights_0=0.8, ConstraintWeights_1=0.2) {
    alphaBoolBlend {blendOption=Linear}
    alphaScaleBiasClamp {bInterpResult=true, interpSpeedDecreasing=5, interpSpeedIncreasing=7}
    boneToModify {boneName=dyn_back_r_chain_4_jnt}
    constraintSetup [{targetBone {boneName=spine_04}}, {targetBone {boneName=clavicle_r}}]
    constraintWeights [1,1]
   }
   n49 = "Transform (Modify) Bone - Bone: master_back_chain_jnt"(ComponentPose=@n48, Alpha=1, Translation="3,-1.076,0", Rotation="-4.851,14.051,-1.213") {
    rotationMode=BMM_Additive
    rotationSpace=BCS_BoneSpace
    translationMode=BMM_Additive
    translationSpace=BCS_BoneSpace
    alphaBoolBlend {blendOption=Linear}
    alphaScaleBiasClamp {bInterpResult=true, interpSpeedDecreasing=5, interpSpeedIncreasing=7}
    boneToModify {boneName=master_back_chain_jnt}
    rotation=<rot -4.851 14.0511 -1.213>
    translation=<3 -1.076>
   }
   n50 = RigidBody(ComponentPose=Local To Component(LocalPose=Use cached pose 'PlatBlue'()), OverrideWorldGravity="0,0,-980") {
    bClampLinearTranslationLimitToRefPose=true
    bOverrideWorldGravity=true
    overridePhysicsAsset=&/F_MED_Platinum_Blue_FaceAcc_Backbling_Physics
    simulationSpace=BaseBoneSpace
    alphaBoolBlend {blendOption=Linear}
    baseBoneRef {boneName=neck_01}
    componentAppliedLinearAccClamp=<3000 3000 300>
    componentLinearAccScale=<0.35 0.35 0.25>
    componentLinearVelScale=<0.35 0.3 0.2>
    overrideWorldGravity {z=-980}
    simSpaceSettings {
     maxAngularAcceleration=1000
     maxAngularVelocity=3
     maxLinearAcceleration=700
     maxLinearVelocity=500
     velocityScaleZ=0
     worldAlpha=1
     externalLinearDragV=<0.2 0.2>
    }
   }
   n51 = "Transform (Modify) Bone - Bone: master_back_chain_jnt"!(ComponentPose=@n50, bAlphaBoolEnabled:bool=Get Cape() -> (Cape:bool), Translation="0,0,0", Rotation="0,0,0", Scale:Vector="1,0.7,1") -> (Pose) {
    alphaInputType=Bool
    scaleMode=BMM_Additive
    scaleSpace=BCS_BoneSpace
    alphaBoolBlend {blendOption=Linear}
    boneToModify {boneName=master_back_chain_jnt}
    scale {y=0.7}
   }
   n52 = "Constraint - dyn_back_l_chain_1_jnt (P)"(ComponentPose=@n51, Alpha=0.5, ConstraintWeights_0=1) {
    alphaBoolBlend {blendOption=Linear}
    boneToModify {boneName=dyn_back_l_chain_1_jnt}
    constraintSetup [{targetBone {boneName=spine_05}}]
    constraintWeights [1]
   }
   n53 = "Constraint - dyn_back_l_chain_2_jnt (P)"(ComponentPose=@n52, Alpha=0.4, ConstraintWeights_0=1) {
    alphaBoolBlend {blendOption=Linear}
    boneToModify {boneName=dyn_back_l_chain_2_jnt}
    constraintSetup [{targetBone {boneName=spine_05}}]
    constraintWeights [1]
   }
   n54 = "Constraint - dyn_back_l_chain_3_jnt (P)"(ComponentPose=@n53, Alpha=0.3, ConstraintWeights_0=1) {
    alphaBoolBlend {blendOption=Linear}
    boneToModify {boneName=dyn_back_l_chain_3_jnt}
    constraintSetup [{targetBone {boneName=spine_05}}]
    constraintWeights [1]
   }
   n55 = "Constraint - dyn_back_l_chain_4_jnt (P)"(ComponentPose=@n54, Alpha=0.2, ConstraintWeights_0=1) {
    alphaBoolBlend {blendOption=Linear}
    boneToModify {boneName=dyn_back_l_chain_4_jnt}
    constraintSetup [{targetBone {boneName=spine_05}}]
    constraintWeights [1]
   }
   n56 = "Constraint - dyn_back_chain_1_jnt (P)"(ComponentPose=@n55, Alpha=0.5, ConstraintWeights_0=1) {
    alphaBoolBlend {blendOption=Linear}
    boneToModify {boneName=dyn_back_chain_1_jnt}
    constraintSetup [{targetBone {boneName=spine_05}}]
    constraintWeights [1]
   }
   n57 = "Constraint - dyn_back_chain_2_jnt (P)"(ComponentPose=@n56, Alpha=0.4, ConstraintWeights_0=1) {
    alphaBoolBlend {blendOption=Linear}
    boneToModify {boneName=dyn_back_chain_2_jnt}
    constraintSetup [{targetBone {boneName=spine_05}}]
    constraintWeights [1]
   }
   n58 = "Constraint - dyn_back_chain_3_jnt (P)"(ComponentPose=@n57, Alpha=0.3, ConstraintWeights_0=1) {
    alphaBoolBlend {blendOption=Linear}
    boneToModify {boneName=dyn_back_chain_3_jnt}
    constraintSetup [{targetBone {boneName=spine_05}}]
    constraintWeights [1]
   }
   n59 = "Constraint - dyn_back_chain_4_jnt (P)"(ComponentPose=@n58, Alpha=0.5, ConstraintWeights_0=1) {
    alphaBoolBlend {blendOption=Linear}
    boneToModify {boneName=dyn_back_chain_4_jnt}
    constraintSetup [{targetBone {boneName=spine_05}}]
    constraintWeights [1]
   }
   n60 = "Constraint - dyn_back_r_chain_1_jnt (P)"(ComponentPose=@n59, Alpha=0.5, ConstraintWeights_0=1) {
    alphaBoolBlend {blendOption=Linear}
    boneToModify {boneName=dyn_back_r_chain_1_jnt}
    constraintSetup [{targetBone {boneName=spine_05}}]
    constraintWeights [1]
   }
   n61 = "Constraint - dyn_back_r_chain_2_jnt (P)"(ComponentPose=@n60, Alpha=0.4, ConstraintWeights_0=1) {
    alphaBoolBlend {blendOption=Linear}
    boneToModify {boneName=dyn_back_r_chain_2_jnt}
    constraintSetup [{targetBone {boneName=spine_05}}]
    constraintWeights [1]
   }
   n62 = "Constraint - dyn_back_r_chain_3_jnt (P)"(ComponentPose=@n61, Alpha=0.3, ConstraintWeights_0=1) {
    alphaBoolBlend {blendOption=Linear}
    boneToModify {boneName=dyn_back_r_chain_3_jnt}
    constraintSetup [{targetBone {boneName=spine_05}}]
    constraintWeights [1]
   }
   n63 = "Constraint - dyn_back_r_chain_4_jnt (P)"(ComponentPose=@n62, Alpha=0.2, ConstraintWeights_0=1) {
    alphaBoolBlend {blendOption=Linear}
    boneToModify {boneName=dyn_back_r_chain_4_jnt}
    constraintSetup [{targetBone {boneName=spine_05}}]
    constraintWeights [1]
   }
   n64 = Blend Poses by bool(
    BlendPose_0=Component To Local(ComponentPose=@n63)
    BlendPose_1=Blend Poses by bool(BlendPose_0=Component To Local(ComponentPose=@n7), BlendPose_1=Component To Local(ComponentPose=@n49), bActiveValue=Get bRigidBodyEnabled()) {blendType=Linear}
    bActiveValue=OR Boolean(A=Get Backbling() -> (Backbling:bool), B=Get Cape())
   ) {blendType=Linear}
   n65 = RigidBody!(ComponentPose=Local To Component(LocalPose=Use cached pose 'PlatBlue'()), Alpha, OverrideWorldGravity="0,0,-980") -> (Pose) {
    bClampLinearTranslationLimitToRefPose=true
    bOverrideWorldGravity=true
    overridePhysicsAsset=&/F_MED_Platinum_Blue_FaceAcc_FrontEnd_Physics
    simulationSpace=BaseBoneSpace
    alphaBoolBlend {blendOption=Linear}
    baseBoneRef {boneName=spine_05}
    componentAppliedLinearAccClamp=<3000 3000 300>
    componentLinearAccScale=<0.4 0.4 0.45>
    componentLinearVelScale=<0.35 0.3 0.2>
    externalForce {z=-980}
    overrideWorldGravity {z=-980}
    simSpaceSettings {
     maxAngularAcceleration=1000
     maxAngularVelocity=3
     maxLinearAcceleration=700
     maxLinearVelocity=500
     velocityScaleZ=0
     worldAlpha=1
     externalLinearDragV=<0.2 0.2>
    }
   }
   n66 = Component To Local(ComponentPose=@n65)
   n67 = Get bIsInFrontEnd()
   n68 = "Constraint - dyn_back_chain_1_jnt (R)"!(
    ComponentPose=Local To Component(LocalPose=Blend Poses by bool(BlendPose_0=@n66, BlendPose_1=@n64, bActiveValue=@n67) {blendType=Linear})
    bAlphaBoolEnabled:bool=True
    ConstraintWeights_0=1
   ) -> (Pose) {
    alphaInputType=Bool
    alphaBoolBlend {blendInTime=1, blendOption=Linear, blendOutTime=1}
    boneToModify {boneName=dyn_back_chain_1_jnt}
    constraintSetup [{transformType=Rotation, targetBone {boneName=spine_05}}]
    constraintWeights [1]
   }
   n69 = "Constraint - dyn_back_r_chain_1_jnt (R)"!(ComponentPose=@n68, bAlphaBoolEnabled:bool=True, ConstraintWeights_0=1) -> (Pose) {
    alphaInputType=Bool
    alphaBoolBlend {blendInTime=1, blendOption=Linear, blendOutTime=1}
    boneToModify {boneName=dyn_back_r_chain_1_jnt}
    constraintSetup [{transformType=Rotation, targetBone {boneName=spine_05}}]
    constraintWeights [1]
   }
   n70 = "Constraint - dyn_back_l_chain_1_jnt (R)"!(ComponentPose=@n69, bAlphaBoolEnabled:bool=True, ConstraintWeights_0=1) -> (Pose) {
    alphaInputType=Bool
    alphaBoolBlend {blendInTime=1, blendOption=Linear, blendOutTime=1}
    boneToModify {boneName=dyn_back_l_chain_1_jnt}
    constraintSetup [{transformType=Rotation, targetBone {boneName=spine_05}}]
    constraintWeights [1]
   }
   n71 = "Constraint - dyn_back_chain_2_jnt (P)"!(ComponentPose=@n70, bAlphaBoolEnabled:bool=True, ConstraintWeights_0=1) -> (Pose) {
    alphaInputType=Bool
    alphaBoolBlend {blendInTime=1, blendOption=Linear, blendOutTime=1}
    boneToModify {boneName=dyn_back_chain_2_jnt}
    constraintSetup [{targetBone {boneName=spine_05}}]
    constraintWeights [1]
   }
   n72 = "Constraint - dyn_back_r_chain_2_jnt (P)"!(ComponentPose=@n71, bAlphaBoolEnabled:bool=True, ConstraintWeights_0=1) -> (Pose) {
    alphaInputType=Bool
    alphaBoolBlend {blendInTime=1, blendOption=Linear, blendOutTime=1}
    boneToModify {boneName=dyn_back_r_chain_2_jnt}
    constraintSetup [{targetBone {boneName=spine_05}}]
    constraintWeights [1]
   }
   n73 = "Constraint - dyn_back_l_chain_2_jnt (P)"!(ComponentPose=@n72, bAlphaBoolEnabled:bool=True, ConstraintWeights_0=1) -> (Pose) {
    alphaInputType=Bool
    alphaBoolBlend {blendInTime=1, blendOption=Linear, blendOutTime=1}
    boneToModify {boneName=dyn_back_l_chain_2_jnt}
    constraintSetup [{targetBone {boneName=spine_05}}]
    constraintWeights [1]
   }
   n74 = "Constraint - dyn_back_chain_3_jnt (R)"!(ComponentPose=@n73, bAlphaBoolEnabled:bool=True, ConstraintWeights_0=1) -> (Pose) {
    alphaInputType=Bool
    alphaBoolBlend {blendInTime=1, blendOption=Linear, blendOutTime=1}
    boneToModify {boneName=dyn_back_chain_3_jnt}
    constraintSetup [{transformType=Rotation, targetBone {boneName=spine_05}}]
    constraintWeights [1]
   }
   n75 = "Constraint - dyn_back_r_chain_3_jnt (R)"!(ComponentPose=@n74, bAlphaBoolEnabled:bool=True, ConstraintWeights_0=1) -> (Pose) {
    alphaInputType=Bool
    alphaBoolBlend {blendInTime=1, blendOption=Linear, blendOutTime=1}
    boneToModify {boneName=dyn_back_r_chain_3_jnt}
    constraintSetup [{transformType=Rotation, targetBone {boneName=spine_05}}]
    constraintWeights [1]
   }
   n76 = "Constraint - dyn_back_l_chain_3_jnt (R)"!(ComponentPose=@n75, bAlphaBoolEnabled:bool=True, ConstraintWeights_0=1) -> (Pose) {
    alphaInputType=Bool
    alphaBoolBlend {blendInTime=1, blendOption=Linear, blendOutTime=1}
    boneToModify {boneName=dyn_back_l_chain_3_jnt}
    constraintSetup [{transformType=Rotation, targetBone {boneName=spine_05}}]
    constraintWeights [1]
   }
   n77 = "Constraint - dyn_back_chain_4_jnt (R)"!(ComponentPose=@n76, bAlphaBoolEnabled:bool=True, ConstraintWeights_0=1) -> (Pose) {
    alphaInputType=Bool
    alphaBoolBlend {blendInTime=1, blendOption=Linear, blendOutTime=1}
    boneToModify {boneName=dyn_back_chain_4_jnt}
    constraintSetup [{transformType=Rotation, targetBone {boneName=spine_05}}]
    constraintWeights [1]
   }
   n78 = "Constraint - dyn_back_r_chain_4_jnt (R)"!(ComponentPose=@n77, bAlphaBoolEnabled:bool=True, ConstraintWeights_0=1) -> (Pose) {
    alphaInputType=Bool
    alphaBoolBlend {blendInTime=1, blendOption=Linear, blendOutTime=1}
    boneToModify {boneName=dyn_back_r_chain_4_jnt}
    constraintSetup [{transformType=Rotation, targetBone {boneName=spine_05}}]
    constraintWeights [1]
   }
   n79 = "Constraint - dyn_back_l_chain_4_jnt (R)"!(ComponentPose=@n78, bAlphaBoolEnabled:bool=True, ConstraintWeights_0=1) -> (Pose) {
    alphaInputType=Bool
    alphaBoolBlend {blendInTime=1, blendOption=Linear, blendOutTime=1}
    boneToModify {boneName=dyn_back_l_chain_4_jnt}
    constraintSetup [{transformType=Rotation, targetBone {boneName=spine_05}}]
    constraintWeights [1]
   }
   n80 = "Transform (Modify) Bone - Bone: master_back_chain_jnt"!(ComponentPose=@n79, Alpha:float=1, Translation="-2,-5,0", Rotation="-0.000,1.000,-0.000") -> (Pose) {
    rotationMode=BMM_Additive
    rotationSpace=BCS_BoneSpace
    translationMode=BMM_Additive
    translationSpace=BCS_BoneSpace
    alphaBoolBlend {blendOption=Linear}
    alphaScaleBiasClamp {bInterpResult=true, interpSpeedDecreasing=5, interpSpeedIncreasing=7}
    boneToModify {boneName=master_back_chain_jnt}
    rotation {roll=0, yaw=1}
    translation=<-2 -5>
   }
   n81 = Look At - Bone: dyn_back_l_chain_1_jnt(ComponentPose=@n80, Alpha=1) {
    alphaBoolBlend {blendOption=Linear}
    boneToModify {boneName=dyn_back_l_chain_1_jnt}
    lookAtLocation=<-27.5968 -19.2707 -10.6482>
    lookAtTarget {boneReference {boneName=spine_05}}
    lookAt_Axis {axis=<1 0>}
   }
   n82 = "Transform (Modify) Bone - Bone: dyn_back_chain_1_jnt"!(ComponentPose=@n81, Alpha=1, Translation="0,0,0", Rotation="-0.000,-30.000,0.000") -> (Pose) {
    rotationMode=BMM_Additive
    rotationSpace=BCS_BoneSpace
    translationMode=BMM_Additive
    translationSpace=BCS_BoneSpace
    alphaBoolBlend {blendOption=Linear}
    alphaScaleBiasClamp {bInterpResult=true, interpSpeedDecreasing=5, interpSpeedIncreasing=7}
    boneToModify {boneName=dyn_back_chain_1_jnt}
    rotation=<rot 0 -30 0>
   }
   n83 = Look At - Bone: dyn_back_r_chain_1_jnt(ComponentPose=@n82, Alpha=1) {
    alphaBoolBlend {blendOption=Linear}
    boneToModify {boneName=dyn_back_r_chain_1_jnt}
    lookAtLocation=<-27.5968 -9.5034 5.9128>
    lookAtTarget {boneReference {boneName=spine_05}}
    lookAt_Axis {axis=<1 0>}
   }
   Output Pose(Result=Component To Local(ComponentPose=@n83)) {layerGroup=None}
   n84 = "Transform (Modify) Bone - Bone: master_back_chain_jnt"!(
    ComponentPose=Local To Component(LocalPose=Copy Pose From Mesh() {bUseAttachedParent=true})
    bAlphaBoolEnabled:bool=Get bIsSkyDiving()
    Translation="-5,0,0"
    Rotation="0,-35,0"
    Scale:Vector="1,1,1"
   ) -> (Pose) {
    alphaInputType=Bool
    rotationMode=BMM_Additive
    rotationSpace=BCS_BoneSpace
    translationMode=BMM_Additive
    translationSpace=BCS_BoneSpace
    alphaBoolBlend {blendOption=Linear}
    boneToModify {boneName=master_back_chain_jnt}
    rotation {yaw=-35}
    translation {x=-5}
   }
   n85 = Get Head R Rot() -> (Head R Rot:double)
   n86 = Look At - Bone: master_back_chain_jnt(ComponentPose=@n84, Alpha=@n85) {
    alphaBoolBlend {blendOption=Linear}
    boneToModify {boneName=master_back_chain_jnt}
    lookAtLocation=<20.6003 -20.933 -4.4782>
    lookAtTarget {boneReference {boneName=pelvis}}
    lookAt_Axis {axis=<1 0>}
   }
   n87 = "Transform (Modify) Bone - Bone: master_back_chain_jnt"!(ComponentPose=@n86, Alpha:float=@n85, Translation="-5,0,0", Rotation="0,50,0", Scale="1,1,1") -> (Pose) {
    rotationMode=BMM_Additive
    rotationSpace=BCS_BoneSpace
    translationMode=BMM_Additive
    translationSpace=BCS_BoneSpace
    alphaBoolBlend {blendOption=Linear}
    boneToModify {boneName=master_back_chain_jnt}
    rotation {yaw=50}
    translation {x=-5}
   }
   n88 = "Transform (Modify) Bone - Bone: master_back_chain_jnt"(ComponentPose=@n87, Alpha=Get Head Back Rot() -> (Head Back Rot:double), Translation="0,-5,0", Rotation="0,0,0", Scale="1,1,1") {
    rotationMode=BMM_Additive
    rotationSpace=BCS_BoneSpace
    translationMode=BMM_Additive
    translationSpace=BCS_BoneSpace
    alphaBoolBlend {blendOption=Linear}
    boneToModify {boneName=master_back_chain_jnt}
    translation {y=-5}
   }
   n89 = Control Rig(Source=Component To Local(ComponentPose=@n88)) {
    alphaBoolBlend {blendOption=Linear}
    controlRigAssetReference {blueprintRigClass=&/F_MED_Platinum_Blue_FaceAcc_CtrlRig.F_MED_Platinum_Blue_FaceAcc_CtrlRig_C}
   }
   Save cached pose 'PlatBlue'(Pose=@n89) {cacheName=PlatBlue}
  ]}
 EventGraph {nodes [
   Event BlueprintUpdateAnimation()
   > n1 = Sequence!(execute) -> (then_0, then_1, then_2:Exec)
   n2 = GetOwnerSkeletalMeshComponent()
   n3 = BreakRotator(InRot=BreakTransform(InTransform=GetDeltaTransformFromRefPose(self=@n2, BoneName=head, BaseName=spine_05)).Rotation)
   Set Head Back Rot(execute=@n1.then_2, Head Back Rot:double=MapRangeClamped(Value=@n3.Yaw, InRangeA=-10, InRangeB=-55)) -> (Output_Get:double)
   > Set Head R Rot(Head R Rot:double=MapRangeClamped(Value=@n3.Pitch, InRangeA=0, InRangeB=45)) -> (Output_Get:double)
   n4 = BreakTransform(InTransform=GetDeltaTransformFromRefPose(self=GetOwnerSkeletalMeshComponent(), BoneName=head, BaseName=neck_01))
   > Set Head L Rot(Head L Rot:double=MapRangeClamped(Value=BreakRotator(InRot=@n4.Rotation), InRangeA=0, InRangeB=1)) -> (Output_Get:double)
   > Set ADS(ADS:double="To Float (Boolean)"(InBool=Get bIsTargeting())) -> (Output_Get:double)
   n5 = NOT Boolean(A=Get bIsSkyDiving())
   n6 = Get bIsMoving2D()
   > Set move(move:double=SelectFloat(A=1, B=0, bPickA=AND Boolean(A=@n6, B=@n5))) -> (Output_Get:double)
   n7 = Gravity Override State Function DEPRECATED(
    execute=@n1.then_0
    CharacterPartAnimInstance=Self-Reference()
    GameTimeInSeconds=GetGameTimeInSeconds()
    GravityOverrideParams=Get Gravity Override Params()
   )
   > Set GravityOverride(GravityOverride=@n7)
   n8 = CheckForBackbling()
   n9 = OR Boolean(A=@n8, B=@n8.bIsBackblingLong)
   n10 = Set Cape(execute=@n1.then_1, Cape:bool=@n9) -> (Output_Get:bool)
   Set Backbling(execute=[@n11.then, @n10.then], Backbling:bool=@n9) -> (Output_Get:bool)
   n11 = Event OnCharacterCustomizationLoaded()
  ]}
]
variables [
 ADS {type=number}
 Backbling {type=boolean}
 Cape {type=boolean}
 Gravity Override Params {type=GravityOverrideParamsStruct}
 GravityOverride {type=Vector}
 Head Back Rot {type=number}
 Head L Rot {type=number}
 Head R Rot {type=number}
 move {type=number}
 Skel_Mesh {type=SkeletalMeshComponent}
]
